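# Translucent unlit Niagara mesh-particle material for the smoky, fiery trail of the female Cactus Rocker outfit.
# Emissive mixes particle-tinted curve-atlas color, inky-smoke highlights and a looping pulse, tinted by time of day.
# Opacity shapes distorted noise and smoke with edge, bottom, depth, view-angle and range fades.
# Vertex offset bends the mesh toward a Niagara-supplied value, clamped, and adds noise and sine wobble.
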Material /BRCosmetics/Characters/Player/Female/Medium/Bodies/F_MED_Cactus_Rocker/FX/M_Cactus_Rocker_Female_Trail
graph {
 properties {
  bUsedWithNiagaraMeshParticles=true
  blendMode=BLEND_Translucent
  pixelDepthOffsetMode=PDOM_Legacy
  shadingModel=MSM_Unlit
  twoSided=true
  usedShadingModels=Unlit
 }
 nodes [
  n1 = DynamicParameter() {desc=Parameters set by Niagara, parameterIndex=1, paramNames [X,Y,Z,WPOIntensity]}
  n2 = Clamp(
   None=Subtract(
    A=MaterialFunctionCall(X=@n1.X, Y=@n1.X, Z=@n1.X) {materialFunction=&MakeFloat3}.Result
    B=ParticlePositionWS()
   )
   Min=ScalarParameter() {defaultValue=-100, group=WPO, parameterName=Clamp_min}
   Max=ScalarParameter() {defaultValue=100, group=WPO, parameterName=Clamp_max}
  ) {maxDefault=1000, minDefault=-1000}
  n3 = OneMinus(None=TextureCoordinate())
  n4 = Multiply(
   A=Power(Base=ComponentMask(None=@n3) {g=true}, Exp=ScalarParameter() {defaultValue=2, group=WPO, parameterName=WPOBendPower})
   B=@n1.WPOIntensity
  )
  n5 = Multiply(
   A=Add(A=TextureCoordinate(), B=ParticleRandom())
   B=ComponentMask(None=VectorParameter() {group=WPO, parameterName=WPO_UVScale, defaultValue=<rgba 0 1 0 0>}.RGB) {g=true, r=true}
  )
  n6 = ComponentMask(None=VectorParameter() {group=WPO, parameterName=WPO_Speed, defaultValue=<rgba 0 0.5 0 0>}.RGB) {g=true, r=true}
  n7 = LinearInterpolate(Alpha=TextureSample(UVs=Panner(Coordinate=@n5, Speed=@n6) {speedY=0.5}) {samplerType=SAMPLERTYPE_Normal, texture=&T_Weap_Distort_N_02}.R) {constB=1.5}
  n8 = OneMinus(None=ComponentMask(None=TextureCoordinate()) {g=true})
  n9 = Saturate(None=Multiply(A=Power(Base=@n8) {constExponent=1}))
  n10 = OneMinus(None=ComponentMask(None=TextureCoordinate()) {g=true})
  n11 = Clamp(None=Multiply(A=Power(Base=@n10) {constExponent=1})) {maxDefault=0.5}
  n12 = Add(A=ComponentMask(None=TextureCoordinate()) {g=true}, B=Time())
  n13 = Multiply(A=Saturate(None=Sine(None=@n12))) {constB=0.3}
  n14 = Add(
   A=Multiply(A=Multiply(A=@n7, B=@n9), B=Constant3Vector() {constant=<rgba 1.5 0 0 1>})
   B=Multiply(A=Multiply(A=@n11, B=@n13), B=Constant3Vector() {constant=<rgba 0 0 1 1>})
  ) {constB=0}
  Add(
   A=StaticSwitchParameter(True=LinearInterpolate(B=@n2, Alpha=@n4), False=Constant()) {defaultValue=true, group=WPO, parameterName=Use_Interpolate WPO}
   B=Multiply(A=Transform(None=@n14)) {constB=13}
  )
  n15 = VectorParameter() {parameterName=Color_Morning, defaultValue=<rgba 0.368 0.1625 0.4271 1>}
  n16 = Constant() {r=1}
  n17 = VectorParameter() {parameterName=Color_Evening, defaultValue=<rgba 0.3054 0.1556 0.3385 1>}
  n18 = VectorParameter() {parameterName=Color_Night, defaultValue=<rgba 0.6152 0.625 0.6055 0.7>}
  n19 = TextureCoordinate() {vTiling=0.5}
  n20 = ComponentMask(None=TextureSample() {samplerType=SAMPLERTYPE_Normal, texture=&T_Weap_Distort_N_02}.RGB) {g=true, r=true}
  n21 = Panner(Coordinate=Add(
    A=MaterialFunctionCall(XY=@n19) {materialFunction=&Swizzle}.YX
    B=Multiply(A=@n20) {constB=0.14}
   )) {speedX=0.4}
  n22 = Multiply(
   A=Power(Base=TextureSample(UVs=@n21) {samplerType=SAMPLERTYPE_Grayscale, texture=&T_Smoke_Trail_Soft}.R) {constExponent=1.6}
   B=Constant() {r=3}
  ) {constB=3}
  n23 = Constant()
  n24 = Time()
  n25 = LinearInterpolate(A=Constant() {r=-2.5}, B=Constant() {r=0.7}, Alpha=Frac(None=Multiply(A=@n24) {constB=1.8}))
  n26 = TextureSample(UVs=Add(A=TextureCoordinate() {uTiling=0, vTiling=1.5}, B=AppendVector(A=@n23, B=@n25)) {constB=-3}) {samplerSource=SSM_Clamp_WorldGroupSettings, texture=/Game/Effects/Fort_Effects/Textures/Smoke/T_Smoke_Trail_Soft_2}
  n27 = Constant3Vector() {constant=<rgba 8.6705 4.396 30 1>}
  n28 = LinearInterpolate(
   A=Constant()
   B=Multiply(A=@n22, B=Multiply(A=@n26.R, B=@n27))
   Alpha=DynamicParameter() {defaultValue=<rgba 1 1 1.5 1>, paramNames [Offset, Highlight_brightness, PulseIntensity, Trail_range]}.PulseIntensity
  )
  n29 = CurveAtlasRowParameter(CurveTime=ComponentMask(None=TextureCoordinate()) {g=true}) {
   atlas=/Game/Athena/LinearColorCurves/A_Atlas_Turtleneck_Glider
   curve=/Game/Characters/Player/Female/Medium/Bodies/F_MED_Cactus_Rocker/FX/Cactus_Rocker_TrailCurve
   defaultValue=6
   parameterName=Curve_Color
  }
  n30 = VectorParameter() {parameterName=Highlight, defaultValue=<rgba 80.5417 12.5 200 1>}
  n31 = DynamicParameter() {defaultValue=<rgba 1 1 1.5 1>, paramNames [Offset, Highlight_brightness, PulseIntensity, Trail_range]}
  n32 = TextureCoordinate() {vTiling=0.2}
  n33 = ParticleRandom()
  n34 = ComponentMask(None=@n32) {g=true}
  n35 = Panner(Coordinate=TextureCoordinate() {uTiling=0.5}) {speedY=0.1}
  n36 = Add(
   A=AppendVector(A=ComponentMask(None=@n32) {r=true}, B=Add(A=@n33, B=@n34))
   B=Multiply(A=TextureSample(UVs=@n35) {samplerType=SAMPLERTYPE_Normal, texture=&T_Weap_Distort_N_02}.R) {constB=0.1}
  )
  n37 = LinearInterpolate(Alpha=TextureSample(UVs=Panner(Coordinate=@n36) {speedY=0.3}) {texture=&T_Inky_Smoke_Tile}.R) {constA=0.2}
  n38 = Add(
   A=Multiply(A=ParticleColor().RGB, B=@n29)
   B=Multiply(A=Multiply(A=@n30.RGB, B=@n31.Highlight_brightness), B=Power(Base=@n37) {constExponent=8})
  ) {constB=0}
  MaterialFunctionCall(Emissive=Multiply(
    A=Add(A=@n28, B=@n38)
    B=MaterialFunctionCall(Color_Morning=@n15.RGB, Color_Day=@n16, Color_Evening=@n17.RGB, Color_Night=@n18.RGB) {materialFunction=&Color_Time_of_Day}.Result
   )) {materialFunction=&MF_FX_EmissiveAdjustments}
  n39 = Add(
   A=DynamicParameter() {defaultValue=<rgba 1 1 1.5 1>, paramNames [Offset, Highlight_brightness, PulseIntensity, Trail_range]}.Offset
   B=TextureCoordinate() {vTiling=1.5}
  )
  n40 = ComponentMask(None=VectorParameter() {group=Distortion, parameterName=Distortion UV Scale, sortPriority=3, defaultValue=<rgba 1 1 0 0>}.RGB) {g=true, r=true}
  n41 = TextureSampleParameter2D(UVs=Panner(Coordinate=Multiply(A=@n39, B=@n40)) {speedY=1}) {
   group=Distortion
   parameterName=Distortion Normal
   samplerType=SAMPLERTYPE_Normal
   sortPriority=2
   texture=&T_Weap_Distort_N_02
  }
  n42 = ComponentMask(None=@n41.RGB) {g=true, r=true}
  n43 = TextureCoordinate()
  n44 = VectorParameter() {group=Fire Shape, parameterName=Fire_MacroShape_UVScale, sortPriority=5, defaultValue=<rgba 1 1 0 0>}
  n45 = ComponentMask(None=@n44.RGB) {g=true, r=true}
  n46 = ComponentMask(None=Multiply(A=@n43, B=@n45)) {g=true}
  n47 = LinearInterpolate(
   B=Divide(
    A=LinearInterpolate(A=@n42, Alpha=@n46) {constB=0}
    B=ScalarParameter() {defaultValue=3, group=Distortion, parameterName=UV_Distortion_Divsor, sortPriority=1}
   ) {constB=20}
   Alpha=ScalarParameter() {defaultValue=1, group=Distortion, parameterName=UV_Distortion_Lerp}
  )
  n48 = Add(
   A=DynamicParameter() {defaultValue=<rgba 1 1 1.5 1>, paramNames [Offset, Highlight_brightness, PulseIntensity, Trail_range]}.Offset
   B=TextureCoordinate() {uTiling=0.8, vTiling=0.9}
  )
  n49 = ComponentMask(None=VectorParameter() {group=Fire Shape, parameterName=Fire_Micro_UVScale, sortPriority=8, defaultValue=<rgba 1 1 0 0>}.RGB) {g=true, r=true}
  n50 = Panner(Coordinate=MaterialFunctionCall(XY=Multiply(A=@n48, B=@n49)) {materialFunction=&Swizzle}.YX) {speedX=1.2, speedY=0.2}
  n51 = TextureSampleParameter2D(UVs=ShadingPathSwitch(Default=Add(A=@n47, B=@n50) {constB=0}, Mobile=@n50) {desc=Migrated from FeatureLevelSwitch}) {
   group=Fire Shape
   parameterName=Fire Micro Texture
   samplerType=SAMPLERTYPE_LinearColor
   sortPriority=6
   texture=&T_NoisePack_LinearColor
  }
  n52 = TextureSample() {samplerType=SAMPLERTYPE_LinearGrayscale, texture=/Engine/Functions/Engine_MaterialFunctions02/ExampleContent/Textures/LowResBlurredNoise}
  n53 = MaterialFunctionCall(Float2=Multiply(A=@n43, B=@n45)) {materialFunction=&BreakOutFloat2Components}
  n54 = Power(
   Base=Sine(None=Multiply(A=@n53.R) {constB=0.5})
   Exp=LinearInterpolate(
    A=ScalarParameter() {defaultValue=6, group=Fire Shape, parameterName=Edge Fade Value Top, sortPriority=3}
    B=ScalarParameter() {defaultValue=2, group=Fire Shape, parameterName=Edge Fade Value Bottom, sortPriority=4}
    Alpha=@n53.G
   ) {constA=6, constB=2}
  ) {constExponent=4}
  n55 = VectorParameter() {group=BottomFade, parameterName=Bottom Fade_SphereMask_A_Scale, sortPriority=2, defaultValue=<rgba 1.2 1 0 0>}
  n56 = Multiply(A=TextureCoordinate(), B=ComponentMask(None=@n55.RGB) {g=true, r=true})
  n57 = VectorParameter() {group=BottomFade, parameterName=Bottom Fade_SphereMask_B, sortPriority=3, defaultValue=<rgba 0.6 0.05 0 0>}
  n58 = ComponentMask(None=@n57.RGB) {g=true, r=true}
  n59 = ScalarParameter() {defaultValue=0.95, group=BottomFade, parameterName=BottomFade_Radius, sortPriority=1}
  n60 = Saturate(None=Multiply(A=SphereMask(A=@n56, B=@n58, Radius=@n59) {attenuationRadius=0.95, hardnessPercent=0}) {constB=15})
  n61 = ScalarParameter() {defaultValue=1, group=BottomFade, parameterName=Bottom Fade}
  n62 = Multiply(
   A=LinearInterpolate(
    A=ChannelMaskParameter(None=@n51.RGBA) {group=Fire Shape, parameterName=Fire Micro Mask, sortPriority=7}.R
    B=@n52.R
    Alpha=ScalarParameter() {defaultValue=0.2, parameterName=Fire Micro lerp}
   ) {constAlpha=0.2}
   B=Multiply(A=Multiply(A=@n54, B=@n53.G), B=LinearInterpolate(B=@n60, Alpha=@n61) {constA=1})
  ) {constA=1}
  n63 = Power(Base=@n62, Exp=LinearInterpolate(Alpha=@n53.G) {constA=2})
  n64 = ScalarParameter() {defaultValue=2, group=Fire Shape, parameterName=Fire Split Contrast, sortPriority=1}
  n65 = OneMinus(None=Sine(None=Multiply(A=@n53.R) {constB=0.45}))
  n66 = LinearInterpolate(Alpha=@n41.R) {constA=0.5, constB=1.5}
  n67 = Round(None=Add(A=Power(Base=@n53.G, Exp=@n64), B=Power(Base=@n65, Exp=@n66) {constExponent=0.5}))
  n68 = ScalarParameter() {defaultValue=0.2, group=Fire Shape, parameterName=Fire Split}
  n69 = Multiply(
   A=Multiply(A=Multiply(A=@n22, B=@n63), B=LinearInterpolate(B=@n67, Alpha=@n68) {constA=1, constB=0})
   B=ScalarParameter() {defaultValue=3, group=Opacity, parameterName=OpacityMaskMultiplier, sortPriority=1}
  ) {constB=3}
  n70 = DepthFade(Opacity=@n69, FadeDistance=ScalarParameter() {defaultValue=3, group=Opacity, parameterName=DepthFadeDistance})
  n71 = Multiply(A=ShadingPathSwitch(Default=Multiply(A=@n70), Mobile=@n69) {desc=Migrated from FeatureLevelSwitch}, B=ParticleColor().A)
  n72 = MaterialFunctionCall("FadeSensitivity (optional)"=ScalarParameter() {defaultValue=1, parameterName=FadeBasedOnViewAngle}) {materialFunction=&FadeBasedOnViewAngle}
  n73 = WorldPosition()
  n74 = ParticlePositionWS()
  n75 = DynamicParameter() {defaultValue=<rgba 1 1 1.5 1>, paramNames [Offset, Highlight_brightness, PulseIntensity, Trail_range]}
  Multiply(
   A=MaterialFunctionCall(Value_Morning=@n15.A, Value_Day=@n16, Value_Evening=@n17.A, Value_Night=@n18.A) {materialFunction=&Scalar_Time_of_Day}.Result
   B=Multiply(A=Multiply(A=@n71, B=@n72.Result), B=SphereMask(A=@n73.XYZ, B=@n74, Radius=@n75.Trail_range) {attenuationRadius=70, hardnessPercent=0.2})
  )
  Power(Base=@n62)
 ]
}
