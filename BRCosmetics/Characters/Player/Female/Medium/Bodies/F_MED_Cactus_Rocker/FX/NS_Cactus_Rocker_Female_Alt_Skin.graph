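# Niagara cosmetic effect for the alternate skin of female Cactus Rocker, attaching music-reactive purple meshes, glow sprites and trails to the body.
# Arm block and ray emitters spawn according to System.Music V, stretching and shifting color with event audio amplitude.
# Persistent arm and back trails feed smoothed position and gameplay-only audio into materials.
# Front-end, time-of-day and distance modifiers tune appearance, while hidden-hand checks drive some visibility.
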NiagaraSystem /BRCosmetics/Characters/Player/Female/Medium/Bodies/F_MED_Cactus_Rocker/FX/NS_Cactus_Rocker_Female_Alt_Skin
dependencies {
 usedDataInterfaces [
  &NiagaraDataInterfaceArrayFloat3
  &NiagaraDataInterfaceSkeletalMesh
  &NiagaraDataInterfaceArrayFloat
  &NiagaraDataInterfaceCurve
 ]
 usedDynamicInputs []
 usedModules [
  &SystemState
  &self:SystemSpawnScript.SystemScriptSource.SystemScriptGraph.NiagaraNodeAssignment_0.SetVariables_#1
 ]
}
emitters {
 Blocks_Right_Arm {
  emitterUpdateScript [
   &EmitterState(
    Inactive Response=::NewEnumerator1
    Life Cycle Mode=::NewEnumerator1
    Loop Behavior=::NewEnumerator1
    Loop Duration Mode=::NewEnumerator1
    UseLoopDelay=false
   )
   &SpawnRate(
    SpawnRate=&MakeCustomFloatFromBool(Bool=&SetBoolByFloatComparison(A=$System.Music V, B=0), True Float=60, False Float=0)
    Use Spawn Probability=false
   )
  ]
  particleSpawnScript [
   &FN_SetSpawnGroup(Number of Spawn Groups=5)
   &InitializeParticle(
    Lifetime=&MakeCustomFloatFromBool(Bool=&SetBoolByFloatComparison(A=$System.Music V, B=0), True Float=0.2, False Float=1)
    Mesh Scale=&Multiply_VectorByFloat(
     Vector=&RandomRangeVector(Minimum=<1 1 1>, Maximum=<1 1 3>)
     Float=&RandomRangeFloat(Minimum=1, Maximum=1.8)
    )
    Mesh Scale Mode=::NewEnumerator4
    Sprite Size Mode=::NewEnumerator3
    UsePositionOffset=false
    Write Material Random=true
   ) {module=InitializeParticle001}
   &SphereLocation(
    MaskSpawnGroup=false
    Override Seed=false
    Sphere Origin:NiagaraPosition=&ConvertVectorToPosition(Input Position=<0 0 0>, Vector Is In Absolute Large World Space:NiagaraBool=false)
    Sphere Radius=1
    Surface Only=false
    UseNonUniformScale=false
    UseOffset=false
    UseSphereOrientation=false
   )
  ]
  particleUpdateScript [
   &ParticleState()
   &DynamicMaterialParameters(
    Index 0 Param 1=1
    Index 0 Param 2=1
    Index 0 Param 3=1
    Index 0 Param 4=1
    Module.Param0WriteEnabled=true
    Module.Param1WriteEnabled=true
    Module.Param2WriteEnabled=true
    Module.Param3WriteEnabled=true
   )
   &SolveForcesAndVelocity(Clamp Velocity=false, Limit Acceleration=false)
   &SkeletalMeshLocation(
    Bone / Socket Index=2
    Bone Sampling Mode=::NewEnumerator4
    MaskSpawnGroup=false
    Orientation Sampling=::NewEnumerator1
    Sampled Position Offset=&Add_Vector(
     A=&ConvertPositionToVector(Input Position=$Particles.Initial.Position, Convert To Absolute Large World Vector=false)
     B=&Add_Vector(A=&RandomRangeVector(Minimum=<-20 0 -5>, Maximum=<2 0 -5>), B=<0 0 0>)
    )
    Skeletal Mesh=DataInterface() {propertyValues {
      DefaultMesh=null
      ExcludeBoneName=root
      OptionalComponentClass=null
      PreviewMesh=/BRCosmetics/Characters/Player/Female/Medium/Bodies/F_MED_Cactus_Rocker/Meshes/F_MED_Cactus_Rocker
      SkinningMode=SkinOnTheFly
      SoftSourceActor=null
      SourceMode=Default
      UvSetIndex=0
      WholeMeshLOD=-1
      bExcludeBone=false
      bReadDeformedGeometry=true
      bRequireCurrentFrameData=true
      ComponentTags []
      FilteredBones [upperarm_l,lowerarm_l,upperarm_r,thigh_l,thigh_r]
      FilteredSockets []
      MeshUserParameter {parameter {name=None, typeDefHandle {registeredTypeIndex=1}, varData []}}
      SamplingRegions []
     }}
   )
   &ScaleColor(
    Scale Alpha=&FloatFromCurve(
     FloatCurve=DataInterface() {propertyValues {Curve {keys [
         {interpMode=RCIM_Cubic, tangentMode=RCTM_User, time=0, value=0}
         {interpMode=RCIM_Cubic, tangentMode=RCTM_User, time=0.5, value=1}
         {tangentMode=RCTM_User, time=1, value=0}
        ]}}}
     Scale Curve=&FN_FrontEndFloatModifier(
      Gameplay=&Multiply_Float(
       A=1
       B=&ClampFloat(Float=&FN_LOD_Distance_Remap(Start Distance=200, End Distance=300, Minimum Output=0, Maximum Output=1))
      )
      Front End=1
     )
    )
    Scale RGB=&Multiply_VectorByFloat(
     Vector=&MakeVectorFromLinearColorRGB(Color=&LerpLinearColors(
       StartColor=&MakeCustomLinearColorFromBool(Bool=&RandomBool(), True Color=<rgba 0.0581 0.0309 0.1979 0.5>, False Color=<rgba 0.1005 0.1436 0.5677 0.5>)
       EndColor=&MakeCustomLinearColorFromBool(
        Bool=$User.bIsFrontEndPreview
        True Color=&MakeCustomLinearColorFromBool(Bool=&RandomBool(), True Color=<rgba 0.8616 0 1 1>, False Color=<rgba 0.0381 0.0078 0.5 1>)
        False Color=&MakeCustomLinearColorFromBool(Bool=&RandomBool(), True Color=<rgba 0.3571 0 0.4 1>, False Color=<rgba 0.0381 0 0.1406 1>)
       )
       LerpFactor=&FN_FrontEndFloatModifier(Gameplay=$NPC.FortniteNPC.EventAudio_Amplitude, Front End=$NPC.FortniteNPC.EventAudio_Amplitude)
      ))
     Float=&Multiply_Float(
      A=&Lerp_Float(
       A=1
       B=&RandomRangeFloat(Minimum=1, Maximum=&RandomRangeFloat(Minimum=7, Maximum=10), Evaluation Type=::NewEnumerator1)
       Alpha=&ClampFloat(Float=&FN_FrontEndFloatModifier(Gameplay=$NPC.FortniteNPC.EventAudio_Amplitude, Front End=$NPC.FortniteNPC.EventAudio_Amplitude))
      )
      B=&FNTimeOfDayScalar(Night Scale=1, Morning Scale=0.5, Evening Scale=0.5)
     )
    )
    ScaleA=true
    ScaleRGB=true
   ) {module=ScaleColor002}
   &ScaleMeshSize(Scale Factor=&MakeVector(
     X=1
     Y=1
     Z=&FloatFromCurve(
      FloatCurve=DataInterface() {propertyValues {Curve {keys [
          {interpMode=RCIM_Cubic, tangentMode=RCTM_User, time=0, value=0}
          {interpMode=RCIM_Cubic, tangentMode=RCTM_User, time=0.5, value=1}
          {tangentMode=RCTM_User, time=1, value=0}
         ]}}}
      Scale Curve=&Lerp_Float(
       A=1.2
       B=&FN_FrontEndFloatModifier(Gameplay=3.8, Front End=3.7)
       Alpha=&ClampFloat(Float=&FN_FrontEndFloatModifier(Gameplay=$NPC.FortniteNPC.EventAudio_Amplitude, Front End=$NPC.FortniteNPC.EventAudio_Amplitude))
      )
     )
    ))
  ]
  rendererClasses [&NiagaraMeshRendererProperties]
  renderers [{rendererClass=&NiagaraMeshRendererProperties, rendererIndex=0}]
 }
 Blocks_Right_Ray {
  emitterUpdateScript [
   &EmitterState(
    Inactive Response=::NewEnumerator1
    Life Cycle Mode=::NewEnumerator1
    Loop Behavior=::NewEnumerator1
    Loop Duration Mode=::NewEnumerator1
    UseLoopDelay=false
   )
   &SpawnRate(
    SpawnRate=&MakeCustomFloatFromBool(Bool=&SetBoolByFloatComparison(A=$System.Music V, B=0), True Float=40, False Float=0)
    Use Spawn Probability=false
   )
  ]
  particleSpawnScript [
   &FN_SetSpawnGroup(Number of Spawn Groups=5)
   &InitializeParticle(
    Color=<rgba 2 2 2 7>
    Lifetime=&MakeCustomFloatFromBool(Bool=&SetBoolByFloatComparison(A=$System.Music V, B=0), True Float=0.2, False Float=1)
    Mesh Scale=&Multiply_VectorByFloat(
     Vector=&RandomRangeVector(Minimum=<0.3 0.3 0.4>, Maximum=<0.3 0.3 0.8>)
     Float=&RandomRangeFloat(Minimum=1, Maximum=1)
    )
    Mesh Scale Mode=::NewEnumerator4
    Sprite Size Mode=::NewEnumerator3
    UsePositionOffset=false
    Write Material Random=true
   ) {module=InitializeParticle001}
   &SphereLocation(
    MaskSpawnGroup=false
    Override Seed=false
    Sphere Origin:NiagaraPosition=&ConvertVectorToPosition(Input Position=<0 0 0>, Vector Is In Absolute Large World Space:NiagaraBool=false)
    Sphere Radius=1
    Surface Only=false
    UseNonUniformScale=false
    UseOffset=false
    UseSphereOrientation=false
   )
  ]
  particleUpdateScript [
   &ParticleState()
   &DynamicMaterialParameters(
    Index 0 Param 2=1
    Index 0 Param 3=1
    Index 0 Param 4=1
    Module.Param0WriteEnabled=true
    Module.Param1WriteEnabled=true
    Module.Param2WriteEnabled=true
    Module.Param3WriteEnabled=true
   )
   &SolveForcesAndVelocity(Clamp Velocity=false, Limit Acceleration=false)
   &SkeletalMeshLocation(
    Bone / Socket Index=2
    Bone Sampling Mode=::NewEnumerator4
    MaskSpawnGroup=false
    Orientation Sampling=::NewEnumerator1
    Sampled Position Offset=&Add_Vector(
     A=&ConvertPositionToVector(Input Position=$Particles.Initial.Position, Convert To Absolute Large World Vector=false)
     B=&Add_Vector(A=&RandomRangeVector(Minimum=<-23 0 -5>, Maximum=<0 0 -6>), B=<0 0 0>)
    )
    Skeletal Mesh=DataInterface() {propertyValues {
      DefaultMesh=null
      ExcludeBoneName=root
      OptionalComponentClass=null
      PreviewMesh=&F_MED_Cactus_Rocker
      SkinningMode=SkinOnTheFly
      SoftSourceActor=null
      SourceMode=Default
      UvSetIndex=0
      WholeMeshLOD=-1
      bExcludeBone=false
      bReadDeformedGeometry=true
      bRequireCurrentFrameData=true
      ComponentTags []
      FilteredBones [upperarm_l,lowerarm_l,upperarm_r,thigh_l,thigh_r]
      FilteredSockets []
      MeshUserParameter {parameter {name=None, typeDefHandle {registeredTypeIndex=1}, varData []}}
      SamplingRegions []
     }}
   )
   &ScaleColor(
    Scale Alpha=&FloatFromCurve(
     FloatCurve=DataInterface() {propertyValues {Curve {keys [
         {interpMode=RCIM_Cubic, tangentMode=RCTM_User, time=0, value=0}
         {interpMode=RCIM_Cubic, tangentMode=RCTM_User, time=0.5, value=1}
         {tangentMode=RCTM_User, time=1, value=0}
        ]}}}
     Scale Curve=&Multiply_Float(
      A=&FN_FrontEndFloatModifier(
       Gameplay=&Multiply_Float(
        A=1
        B=&ClampFloat(Float=&FN_LOD_Distance_Remap(Start Distance=200, End Distance=300, Minimum Output=0, Maximum Output=1))
       )
       Front End=1
      )
      B=&FNTimeOfDayScalar(Night Scale=0.7, Morning Scale=0.3, Evening Scale=0.3)
     )
    )
    Scale RGB=&Multiply_VectorByFloat(
     Vector=&MakeVectorFromLinearColorRGB(Color=&LerpLinearColors(
       StartColor=&MakeCustomLinearColorFromBool(Bool=&RandomBool(), True Color=<rgba 0.0581 0.0309 0.1979 0.5>, False Color=<rgba 0.1005 0.1436 0.5677 0.5>)
       EndColor=&MakeCustomLinearColorFromBool(
        Bool=$User.bIsFrontEndPreview
        True Color=&MakeCustomLinearColorFromBool(Bool=&RandomBool(), True Color=<rgba 1.1466 0 1.5 1>, False Color=<rgba 0.1936 0 4 1>)
        False Color=&MakeCustomLinearColorFromBool(Bool=&RandomBool(), True Color=<rgba 0.7644 0 1 1>, False Color=<rgba 0.3423 0 1 1>)
       )
       LerpFactor=&FN_FrontEndFloatModifier(Gameplay=$NPC.FortniteNPC.EventAudio_Amplitude, Front End=$NPC.FortniteNPC.EventAudio_Amplitude)
      ))
     Float=&Lerp_Float(
      A=1
      B=1
      Alpha=&ClampFloat(Float=&FN_FrontEndFloatModifier(Gameplay=$NPC.FortniteNPC.EventAudio_Amplitude, Front End=$NPC.FortniteNPC.EventAudio_Amplitude))
     )
    )
    ScaleA=true
    ScaleRGB=true
   ) {module=ScaleColor002}
   &ScaleMeshSize(Scale Factor=&MakeVector(
     X=1
     Y=1
     Z=&FloatFromCurve(
      FloatCurve=DataInterface() {propertyValues {Curve {keys [
          {interpMode=RCIM_Cubic, tangentMode=RCTM_User, time=0, value=0}
          {interpMode=RCIM_Cubic, tangentMode=RCTM_User, time=0.5, value=1}
          {tangentMode=RCTM_User, time=1, value=0}
         ]}}}
      Scale Curve=&Lerp_Float(
       A=1.2
       B=&FN_FrontEndFloatModifier(Gameplay=2, Front End=1.6)
       Alpha=&ClampFloat(Float=&FN_FrontEndFloatModifier(Gameplay=$NPC.FortniteNPC.EventAudio_Amplitude, Front End=$NPC.FortniteNPC.EventAudio_Amplitude))
      )
     )
    )) {module=ScaleMeshSize001}
  ]
  rendererClasses [&NiagaraMeshRendererProperties]
  renderers [{rendererClass=&NiagaraMeshRendererProperties, rendererIndex=0}]
 }
 Blocks_left_Arm001 {
  emitterUpdateScript [
   &EmitterState(
    Inactive Response=::NewEnumerator1
    Life Cycle Mode=::NewEnumerator1
    Loop Behavior=::NewEnumerator1
    Loop Duration Mode=::NewEnumerator1
    UseLoopDelay=false
   )
   &SpawnRate(
    SpawnRate=&MakeCustomFloatFromBool(Bool=&SetBoolByFloatComparison(A=$System.Music V, B=0), True Float=60, False Float=0)
    Use Spawn Probability=false
   )
  ]
  particleSpawnScript [
   &FN_SetSpawnGroup(Number of Spawn Groups=3)
   &InitializeParticle(
    Lifetime=&MakeCustomFloatFromBool(Bool=&SetBoolByFloatComparison(A=$System.Music V, B=0), True Float=0.2, False Float=1)
    Mesh Scale=&Multiply_VectorByFloat(
     Vector=&RandomRangeVector(Minimum=<1 1 2>, Maximum=<1 1 3>)
     Float=&RandomRangeFloat(Minimum=1, Maximum=1.8)
    )
    Mesh Scale Mode=::NewEnumerator4
    Sprite Size Mode=::NewEnumerator3
    UsePositionOffset=false
    Write Material Random=true
   )
   &SphereLocation(
    MaskSpawnGroup=false
    Override Seed=false
    Sphere Origin:NiagaraPosition=&ConvertVectorToPosition(Input Position=<0 0 0>, Vector Is In Absolute Large World Space:NiagaraBool=false)
    Sphere Radius=1
    Surface Only=false
    UseNonUniformScale=false
    UseOffset=false
    UseSphereOrientation=false
   )
  ]
  particleUpdateScript [
   &ParticleState()
   &DynamicMaterialParameters(
    Index 0 Param 1=1
    Index 0 Param 2=1
    Index 0 Param 3=1
    Index 0 Param 4=1
    Module.Param0WriteEnabled=true
    Module.Param1WriteEnabled=true
    Module.Param2WriteEnabled=true
    Module.Param3WriteEnabled=true
   )
   &SolveForcesAndVelocity(Clamp Velocity=false, Limit Acceleration=false)
   &SkeletalMeshLocation(
    Bone Sampling Mode=::NewEnumerator4
    MaskSpawnGroup=false
    Orientation Sampling=::NewEnumerator1
    Sampled Position Offset=&Add_Vector(
     A=&ConvertPositionToVector(Input Position=$Particles.Initial.Position, Convert To Absolute Large World Vector=false)
     B=&Add_Vector(A=&RandomRangeVector(Minimum=<2 0 5>, Maximum=<25 0 5>), B=<0 0 0>)
    )
    Skeletal Mesh=DataInterface() {propertyValues {
      DefaultMesh=null
      ExcludeBoneName=root
      OptionalComponentClass=null
      PreviewMesh=&F_MED_Cactus_Rocker
      SkinningMode=SkinOnTheFly
      SoftSourceActor=null
      SourceMode=Default
      UvSetIndex=0
      WholeMeshLOD=-1
      bExcludeBone=false
      bReadDeformedGeometry=true
      bRequireCurrentFrameData=true
      ComponentTags []
      FilteredBones [upperarm_l,lowerarm_l,lowerarm_r,thigh_l,thigh_r]
      FilteredSockets []
      MeshUserParameter {parameter {name=None, typeDefHandle {registeredTypeIndex=1}, varData []}}
      SamplingRegions []
     }}
   )
   &ScaleColor(
    Scale Alpha=&FloatFromCurve(
     FloatCurve=DataInterface() {propertyValues {Curve {keys [
         {interpMode=RCIM_Cubic, tangentMode=RCTM_User, time=0, value=0}
         {interpMode=RCIM_Cubic, tangentMode=RCTM_User, time=0.5, value=1}
         {tangentMode=RCTM_User, time=1, value=0}
        ]}}}
     Scale Curve=&FN_FrontEndFloatModifier(
      Gameplay=&Multiply_Float(
       A=1
       B=&ClampFloat(Float=&FN_LOD_Distance_Remap(Start Distance=200, End Distance=300, Minimum Output=0, Maximum Output=1))
      )
      Front End=1
     )
    )
    Scale RGB=&Multiply_VectorByFloat(
     Vector=&MakeVectorFromLinearColorRGB(Color=&LerpLinearColors(
       StartColor=&MakeCustomLinearColorFromBool(Bool=&RandomBool(), True Color=<rgba 0.0581 0.0309 0.1979 0.5>, False Color=<rgba 0.1005 0.1436 0.5677 0.5>)
       EndColor=&MakeCustomLinearColorFromBool(
        Bool=$User.bIsFrontEndPreview
        True Color=&MakeCustomLinearColorFromBool(Bool=&RandomBool(), True Color=<rgba 0.8616 0 1 1>, False Color=<rgba 0.0381 0.0078 0.5 1>)
        False Color=&MakeCustomLinearColorFromBool(Bool=&RandomBool(), True Color=<rgba 0.3571 0 0.4 1>, False Color=<rgba 0.0381 0 0.1406 1>)
       )
       LerpFactor=&FN_FrontEndFloatModifier(Gameplay=$NPC.FortniteNPC.EventAudio_Amplitude, Front End=$NPC.FortniteNPC.EventAudio_Amplitude)
      ))
     Float=&Multiply_Float(
      A=&Lerp_Float(
       A=1
       B=&RandomRangeFloat(Minimum=1, Maximum=&RandomRangeFloat(Minimum=7, Maximum=10), Evaluation Type=::NewEnumerator1)
       Alpha=&ClampFloat(Float=&FN_FrontEndFloatModifier(Gameplay=$NPC.FortniteNPC.EventAudio_Amplitude, Front End=$NPC.FortniteNPC.EventAudio_Amplitude))
      )
      B=&FNTimeOfDayScalar(Night Scale=1, Morning Scale=0.5, Evening Scale=0.5)
     )
    )
    ScaleA=true
    ScaleRGB=true
   ) {module=ScaleColor002}
   &ScaleMeshSize(Scale Factor=&MakeVector(
     X=1
     Y=1
     Z=&FloatFromCurve(
      FloatCurve=DataInterface() {propertyValues {Curve {keys [
          {interpMode=RCIM_Cubic, tangentMode=RCTM_User, time=0, value=0}
          {interpMode=RCIM_Cubic, tangentMode=RCTM_User, time=0.5, value=1}
          {tangentMode=RCTM_User, time=1, value=0}
         ]}}}
      Scale Curve=&Lerp_Float(
       A=1.2
       B=&FN_FrontEndFloatModifier(Gameplay=3.8, Front End=3.7)
       Alpha=&ClampFloat(Float=&FN_FrontEndFloatModifier(Gameplay=$NPC.FortniteNPC.EventAudio_Amplitude, Front End=$NPC.FortniteNPC.EventAudio_Amplitude))
      )
     )
    ))
  ]
  rendererClasses [&NiagaraMeshRendererProperties]
  renderers [{rendererClass=&NiagaraMeshRendererProperties, rendererIndex=0}]
 }
 Blocks_left_Ray {
  emitterUpdateScript [
   &EmitterState(
    Inactive Response=::NewEnumerator1
    Life Cycle Mode=::NewEnumerator1
    Loop Behavior=::NewEnumerator1
    Loop Duration Mode=::NewEnumerator1
    UseLoopDelay=false
   )
   &SpawnRate(
    SpawnRate=&MakeCustomFloatFromBool(Bool=&SetBoolByFloatComparison(A=$System.Music V, B=0), True Float=40, False Float=0)
    Use Spawn Probability=false
   )
  ]
  particleSpawnScript [
   &FN_SetSpawnGroup(Number of Spawn Groups=3)
   &InitializeParticle(
    Color=<rgba 2 2 2 7>
    Lifetime=&MakeCustomFloatFromBool(Bool=&SetBoolByFloatComparison(A=$System.Music V, B=0), True Float=0.2, False Float=1)
    Mesh Scale=&Multiply_VectorByFloat(
     Vector=&RandomRangeVector(Minimum=<0.3 0.3 0.4>, Maximum=<0.3 0.3 0.8>)
     Float=&RandomRangeFloat(Minimum=1, Maximum=1)
    )
    Mesh Scale Mode=::NewEnumerator4
    Sprite Size Mode=::NewEnumerator3
    UsePositionOffset=false
    Write Material Random=true
   )
   &SphereLocation(
    MaskSpawnGroup=false
    Override Seed=false
    Sphere Origin:NiagaraPosition=&ConvertVectorToPosition(Input Position=<0 0 0>, Vector Is In Absolute Large World Space:NiagaraBool=false)
    Sphere Radius=1
    Surface Only=false
    UseNonUniformScale=false
    UseOffset=false
    UseSphereOrientation=false
   )
  ]
  particleUpdateScript [
   &ParticleState()
   &DynamicMaterialParameters(
    Index 0 Param 2=1
    Index 0 Param 3=1
    Index 0 Param 4=1
    Module.Param0WriteEnabled=true
    Module.Param1WriteEnabled=true
    Module.Param2WriteEnabled=true
    Module.Param3WriteEnabled=true
   )
   &SolveForcesAndVelocity(Clamp Velocity=false, Limit Acceleration=false)
   &SkeletalMeshLocation(
    Bone Sampling Mode=::NewEnumerator4
    MaskSpawnGroup=false
    Orientation Sampling=::NewEnumerator1
    Sampled Position Offset=&Add_Vector(
     A=&ConvertPositionToVector(Input Position=$Particles.Initial.Position, Convert To Absolute Large World Vector=false)
     B=&Add_Vector(A=&RandomRangeVector(Minimum=<0 0 5>, Maximum=<25 0 5>), B=<0 0 0>)
    )
    Skeletal Mesh=DataInterface() {propertyValues {
      DefaultMesh=null
      ExcludeBoneName=root
      OptionalComponentClass=null
      PreviewMesh=&F_MED_Cactus_Rocker
      SkinningMode=SkinOnTheFly
      SoftSourceActor=null
      SourceMode=Default
      UvSetIndex=0
      WholeMeshLOD=-1
      bExcludeBone=false
      bReadDeformedGeometry=true
      bRequireCurrentFrameData=true
      ComponentTags []
      FilteredBones [upperarm_l,lowerarm_l,lowerarm_r,thigh_l,thigh_r]
      FilteredSockets []
      MeshUserParameter {parameter {name=None, typeDefHandle {registeredTypeIndex=1}, varData []}}
      SamplingRegions []
     }}
   )
   &ScaleColor(
    Scale Alpha=&FloatFromCurve(
     FloatCurve=DataInterface() {propertyValues {Curve {keys [
         {interpMode=RCIM_Cubic, tangentMode=RCTM_User, time=0, value=0}
         {interpMode=RCIM_Cubic, tangentMode=RCTM_User, time=0.5, value=1}
         {tangentMode=RCTM_User, time=1, value=0}
        ]}}}
     Scale Curve=&Multiply_Float(
      A=&FN_FrontEndFloatModifier(
       Gameplay=&Multiply_Float(
        A=1
        B=&ClampFloat(Float=&FN_LOD_Distance_Remap(Start Distance=200, End Distance=300, Minimum Output=0, Maximum Output=1))
       )
       Front End=1
      )
      B=&FNTimeOfDayScalar(Night Scale=0.7, Morning Scale=0.3, Evening Scale=0.3)
     )
    )
    Scale RGB=&Multiply_VectorByFloat(
     Vector=&MakeVectorFromLinearColorRGB(Color=&LerpLinearColors(
       StartColor=&MakeCustomLinearColorFromBool(Bool=&RandomBool(), True Color=<rgba 0.0581 0.0309 0.1979 0.5>, False Color=<rgba 0.1005 0.1436 0.5677 0.5>)
       EndColor=&MakeCustomLinearColorFromBool(
        Bool=$User.bIsFrontEndPreview
        True Color=&MakeCustomLinearColorFromBool(Bool=&RandomBool(), True Color=<rgba 1.1466 0 1.5 1>, False Color=<rgba 0.1936 0 4 1>)
        False Color=&MakeCustomLinearColorFromBool(Bool=&RandomBool(), True Color=<rgba 0.7644 0 1 1>, False Color=<rgba 0.3423 0 1 1>)
       )
       LerpFactor=&FN_FrontEndFloatModifier(Gameplay=$NPC.FortniteNPC.EventAudio_Amplitude, Front End=$NPC.FortniteNPC.EventAudio_Amplitude)
      ))
     Float=&Lerp_Float(
      A=1
      B=1
      Alpha=&ClampFloat(Float=&FN_FrontEndFloatModifier(Gameplay=$NPC.FortniteNPC.EventAudio_Amplitude, Front End=$NPC.FortniteNPC.EventAudio_Amplitude))
     )
    )
    ScaleA=true
    ScaleRGB=true
   )
   &ScaleMeshSize(Scale Factor=&MakeVector(
     X=1
     Y=1
     Z=&FloatFromCurve(
      FloatCurve=DataInterface() {propertyValues {Curve {keys [
          {interpMode=RCIM_Cubic, tangentMode=RCTM_User, time=0, value=0}
          {interpMode=RCIM_Cubic, tangentMode=RCTM_User, time=0.5, value=1}
          {tangentMode=RCTM_User, time=1, value=0}
         ]}}}
      Scale Curve=&Lerp_Float(
       A=1.2
       B=&FN_FrontEndFloatModifier(Gameplay=2, Front End=1.6)
       Alpha=&ClampFloat(Float=&FN_FrontEndFloatModifier(Gameplay=$NPC.FortniteNPC.EventAudio_Amplitude, Front End=$NPC.FortniteNPC.EventAudio_Amplitude))
      )
     )
    ))
  ]
  rendererClasses [&NiagaraMeshRendererProperties]
  renderers [{rendererClass=&NiagaraMeshRendererProperties, rendererIndex=0}]
 }
 Particle_Decal {
  emitterUpdateScript [
   &EmitterState(
    Inactive Response=::NewEnumerator1
    Life Cycle Mode=::NewEnumerator1
    Loop Behavior=::NewEnumerator1
    Loop Duration Mode=::NewEnumerator1
    UseLoopDelay=false
   )
   &SpawnBurst_Instantaneous(Spawn Count=3, Use Spawn Probability=false, UseLoopCountLimit=false)
  ]
  particleSpawnScript [
   &FN_SetSpawnGroup(Number of Spawn Groups=3)
   &InitializeParticle(
    Color=<rgba 1.8164 0.4613 12 1>
    Sprite Size Mode=::NewEnumerator3
    Uniform Sprite Size=&FN_FrontEndFloatModifier(
     Gameplay=&SelectFloatFromArray(Direct Array Index=$Particles.SpawnGroup, Float Selection Array=DataInterface() {propertyValues {FloatData [20,0,20]}})
     Front End=&SelectFloatFromArray(Direct Array Index=$Particles.SpawnGroup, Float Selection Array=DataInterface() {propertyValues {FloatData [20,20,20]}})
    )
    UsePositionOffset=false
    Write Material Random=true
   )
   &CameraOffset(Camera Offset Amount=20)
  ]
  particleUpdateScript [
   &ParticleState(Kill Particles When Lifetime Has Elapsed=false, Let Infinitely Lived Particles Die When Emitter Deactivates=true)
   &SkeletalMeshLocation(
    Bone / Socket Index=$Particles.SpawnGroup
    Bone Sampling Mode=::NewEnumerator4
    MaskSpawnGroup=false
    Sampled Position Offset=&SelectVectorFromArray(Direct Array Index=$Particles.SpawnGroup, Vector Selection Array=DataInterface() {propertyValues {FloatData [<0 0 0>, <10 0 0>, <0 0 0>]}})
    Skeletal Mesh=DataInterface() {propertyValues {
      DefaultMesh=null
      ExcludeBoneName=root
      OptionalComponentClass=null
      PreviewMesh=&F_MED_Cactus_Rocker
      SkinningMode=SkinOnTheFly
      SoftSourceActor=null
      SourceMode=Default
      UvSetIndex=0
      WholeMeshLOD=-1
      bExcludeBone=false
      bReadDeformedGeometry=true
      bRequireCurrentFrameData=true
      ComponentTags []
      FilteredBones [upperarm_l,lowerarm_l,lowerarm_r]
      FilteredSockets []
      MeshUserParameter {parameter {name=None, typeDefHandle {registeredTypeIndex=1}, varData []}}
      SamplingRegions []
     }}
   )
   &ScaleColor(
    Scale RGB=&Multiply_VectorByFloat(Vector=<1 1 1>, Float=&FN_FrontEndFloatModifier(Gameplay=1.5, Front End=2))
    ScaleA=true
    ScaleRGB=true
   )
   /SetVariables(Particles.VisibilityTag=&self:PartHiddenBoneCheck(is Hand Hidden:NiagaraBool=$User.bisHandHidden, Bone Selection:NiagaraInt32=$Particles.SpawnGroup)) {isSetParameters=true, module=SetVariables_#2}
  ]
  rendererClasses [&NiagaraSpriteRendererProperties]
  renderers [{rendererClass=&NiagaraSpriteRendererProperties}]
 }
 Trail_arms_and_legs {
  emitterUpdateScript [
   &EmitterState(
    Inactive Response=::NewEnumerator1
    Life Cycle Mode=::NewEnumerator1
    Loop Behavior=::NewEnumerator1
    Loop Duration Mode=::NewEnumerator1
    UseLoopDelay=false
   )
   &SpawnBurst_Instantaneous(Spawn Count=4, Use Spawn Probability=false, UseLoopCountLimit=false)
  ]
  particleSpawnScript [
   &FN_SetSpawnGroup(Number of Spawn Groups=6)
   &InitializeParticle(
    Color=&MakeCustomLinearColorFromBool(Bool=$User.bIsFrontEndPreview, True Color=<rgba 0.6812 0.3226 1 0.3>, False Color=<rgba 1 1 1 0.3>)
    Lifetime=0.1
    Mesh Scale=&Multiply_VectorByFloat(
     Vector=<1 1 1>
     Float=&SelectFloatFromArray(Direct Array Index=$Particles.SpawnGroup, Float Selection Array=DataInterface() {propertyValues {FloatData [0.4,0.4,0.4,0.4]}})
    )
    Mesh Scale Mode=::NewEnumerator4
    UsePositionOffset=false
    Write Material Random=true
   )
   &InitialMeshOrientation(
    Mesh Orientation Mode=::NewEnumerator7
    Rotation=&SelectVectorFromArray(
     Direct Array Index=$Particles.SpawnGroup
     Vector Selection Array=DataInterface() {propertyValues {FloatData [<0 0 0.25>, <0.1 0 0>, <0 0 0.5>, <0 0 0.5>]}}
    )
    Use Rotation Vector=true
   )
   &InitialMeshOrientation(
    Mesh Orientation Mode=::NewEnumerator7
    Rotation=&SelectVectorFromArray(
     Direct Array Index=$Particles.SpawnGroup
     Vector Selection Array=DataInterface() {propertyValues {FloatData [<0 0 0.5>, <0 0 0>, <0 0.33 0>, <0.5 0 0>]}}
    )
    Use Rotation Vector=true
   ) {module=InitialMeshOrientation001}
   &InitialMeshOrientation(
    Mesh Orientation Mode=::NewEnumerator7
    Rotation=&SelectVectorFromArray(
     Direct Array Index=$Particles.SpawnGroup
     Vector Selection Array=DataInterface() {propertyValues {FloatData [<0.3 0 0>, <0.5 0 0>, <0.1 0 0>, <0.3 0 0>]}}
    )
    Use Rotation Vector=true
   ) {module=InitialMeshOrientation002}
   &InitialMeshOrientation(
    Mesh Orientation Mode=::NewEnumerator7
    Rotation=&SelectVectorFromArray(
     Direct Array Index=$Particles.SpawnGroup
     Vector Selection Array=DataInterface() {propertyValues {FloatData [<0.02 -0.02 0>, <0 0.2 0.2>, <0 -0.2 0>, <0 0 0>]}}
    )
    Use Rotation Vector=true
   ) {module=InitialMeshOrientation003}
  ]
  particleUpdateScript [
   &ParticleState(Kill Particles When Lifetime Has Elapsed=false, Let Infinitely Lived Particles Die When Emitter Deactivates=true)
   &InterpolateOverTime(
    Rate Of Change=5
    Target Vector Value:Vector3f=&ConvertPositionToVector(Input Position=$Particles.Position, Convert To Absolute Large World Vector=false)
    Variable Type=::NewEnumerator2
   )
   &DynamicMaterialParameters(
    Index 0 Param 1=1
    Index 0 Param 2=0.5
    Index 0 Param 3=&ClampFloat(
     Float=&Lerp_Float(
      A=0
      B=10
      Alpha=&FN_FrontEndFloatModifier(Gameplay=$NPC.FortniteNPC.EventAudio_Amplitude, Front End=0)
     )
     Max=10
    )
    Index 0 Param 4=60
    Module.Param0WriteEnabled=true
    Module.Param1WriteEnabled=true
    Module.Param2WriteEnabled=true
    Module.Param3WriteEnabled=true
   )
   /SetVariables(Particles.Interpolate Dynamic=&MakeFloat4FromVectorAndFloat("Vector (RGB)"=$StackContext.InterpolateOverTime.InterpolateOverTime_V3_.Moving Average, "Float (Alpha)"=1)) {isSetParameters=true, module=SetVariables_#3}
   &SolveForcesAndVelocity(Clamp Velocity=false, Limit Acceleration=false)
   &SkeletalMeshLocation(
    Bone / Socket Index=$Particles.SpawnGroup
    Bone Sampling Mode=::NewEnumerator4
    MaskSpawnGroup=false
    Sampled Position Offset=&SelectVectorFromArray(
     Direct Array Index=$Particles.SpawnGroup
     Vector Selection Array=DataInterface() {propertyValues {FloatData [<0 0 0>, <15 0 0>, <0 0 0>, <-25 0 0>]}}
    )
    Skeletal Mesh=DataInterface() {propertyValues {
      DefaultMesh=null
      ExcludeBoneName=root
      OptionalComponentClass=null
      PreviewMesh=&F_MED_Cactus_Rocker
      SkinningMode=SkinOnTheFly
      SoftSourceActor=null
      SourceMode=Default
      UvSetIndex=0
      WholeMeshLOD=-1
      bExcludeBone=false
      bReadDeformedGeometry=true
      bRequireCurrentFrameData=true
      ComponentTags []
      FilteredBones [upperarm_l,lowerarm_l,lowerarm_r,upperarm_r]
      FilteredSockets []
      MeshUserParameter {parameter {name=None, typeDefHandle {registeredTypeIndex=1}, varData []}}
      SamplingRegions []
     }}
   )
   &ScaleColor(
    Scale Alpha=&Multiply_Float(A=&FN_FrontEndFloatModifier(Gameplay=1, Front End=3), B=1)
    Scale RGB=&Multiply_VectorByFloat(
     Vector=<1 1 1>
     Float=&Multiply_Float(A=1.2, B=&FN_FrontEndFloatModifier(Gameplay=1, Front End=1.5))
    )
    ScaleA=true
    ScaleRGB=true
   )
   &ScaleMeshSize(Scale Factor=&FN_FrontEndVector3Modifier(Gameplay=<1 1 1.5>, Front End=<1.3 1.3 1.2>))
   /SetVariables(Particles.VisibilityTag=&self:PartHiddenBoneCheck(is Hand Hidden:NiagaraBool=$User.bisHandHidden, Bone Selection:NiagaraInt32=$Particles.SpawnGroup)) {isSetParameters=true, module=SetVariables_#2}
  ]
  rendererClasses [&NiagaraMeshRendererProperties]
  renderers [{rendererClass=&NiagaraMeshRendererProperties, rendererIndex=0}]
 }
 Trail_back {
  emitterUpdateScript [
   &EmitterState(
    Inactive Response=::NewEnumerator1
    Life Cycle Mode=::NewEnumerator1
    Loop Behavior=::NewEnumerator1
    Loop Duration Mode=::NewEnumerator1
    UseLoopDelay=false
   )
   &SpawnBurst_Instantaneous(Spawn Count=2, Use Spawn Probability=false, UseLoopCountLimit=false)
  ]
  particleSpawnScript [
   &FN_SetSpawnGroup(Number of Spawn Groups=3)
   &InitializeParticle(
    Color=<rgba 1 1 1 0.3>
    Mesh Scale=&Multiply_VectorByFloat(Vector=<1 1 1>, Float=&FN_FrontEndFloatModifier(Gameplay=0.5, Front End=0.6))
    Mesh Scale Mode=::NewEnumerator4
    UsePositionOffset=false
    Write Material Random=true
   )
   &InitialMeshOrientation(
    Mesh Orientation Mode=::NewEnumerator7
    Rotation=&RandomRangeVector(Minimum=<-1 -1 -1>, Maximum=<1 1 1>)
    Use Rotation Vector=true
   )
   &InitialMeshOrientation(
    Mesh Orientation Mode=::NewEnumerator7
    Rotation=&SelectVectorFromArray(
     Direct Array Index=$Particles.SpawnGroup
     Vector Selection Array=DataInterface() {propertyValues {FloatData [<0 -0.15 0.2>, <0 0 0>, <0 -0.1 0.25>]}}
    )
    Use Rotation Vector=true
   ) {module=InitialMeshOrientation001}
   &CameraOffset()
  ]
  particleUpdateScript [
   &ParticleState(Kill Particles When Lifetime Has Elapsed=false, Let Infinitely Lived Particles Die When Emitter Deactivates=true)
   &InterpolateOverTime(
    Rate Of Change=5
    Target Vector Value:Vector3f=&ConvertPositionToVector(Input Position=$Particles.Position, Convert To Absolute Large World Vector=false)
    Variable Type=::NewEnumerator2
   )
   &DynamicMaterialParameters(
    Index 0 Param 1=1
    Index 0 Param 2=1
    Index 0 Param 3=&ClampFloat(
     Float=&Lerp_Float(
      A=0
      B=1.5
      Alpha=&FN_FrontEndFloatModifier(Gameplay=$NPC.FortniteNPC.EventAudio_Amplitude, Front End=0)
     )
     Max=1.5
    )
    Index 0 Param 4=60
    Module.Param0WriteEnabled=true
    Module.Param1WriteEnabled=true
    Module.Param2WriteEnabled=true
    Module.Param3WriteEnabled=true
   )
   /SetVariables(Particles.Interpolate Dynamic=&MakeFloat4FromVectorAndFloat("Vector (RGB)"=$StackContext.InterpolateOverTime.InterpolateOverTime_V3_.Moving Average, "Float (Alpha)"=1)) {isSetParameters=true, module=SetVariables_#3}
   &SolveForcesAndVelocity(Clamp Velocity=false, Limit Acceleration=false)
   &SkeletalMeshLocation(
    Bone / Socket Index=$Particles.SpawnGroup
    Bone Sampling Mode=::NewEnumerator4
    MaskSpawnGroup=false
    Sampled Position Offset=&SelectVectorFromArray(
     Direct Array Index=$Particles.SpawnGroup
     Vector Selection Array=DataInterface() {propertyValues {FloatData [<-5 0 -8>, <0 0 7>, <-13 0 -5>]}}
    )
    Skeletal Mesh=DataInterface() {propertyValues {
      DefaultMesh=null
      ExcludeBoneName=root
      OptionalComponentClass=null
      PreviewMesh=&F_MED_Cactus_Rocker
      SkinningMode=SkinOnTheFly
      SoftSourceActor=null
      SourceMode=Default
      UvSetIndex=0
      WholeMeshLOD=-1
      bExcludeBone=false
      bReadDeformedGeometry=true
      bRequireCurrentFrameData=true
      ComponentTags []
      FilteredBones [spine_03,spine_03,spine_04]
      FilteredSockets []
      MeshUserParameter {parameter {name=None, typeDefHandle {registeredTypeIndex=1}, varData []}}
      SamplingRegions []
     }}
   )
   &ScaleColor(
    Scale Alpha=&FN_FrontEndFloatModifier(Gameplay=1, Front End=2)
    Scale RGB=&Multiply_VectorByFloat(Vector=<1 1 1>, Float=&FN_FrontEndFloatModifier(Gameplay=1.2, Front End=1.5))
    ScaleA=true
    ScaleRGB=true
   )
  ]
  rendererClasses [&NiagaraMeshRendererProperties]
  renderers [{rendererClass=&NiagaraMeshRendererProperties, rendererIndex=0}]
 }
}
summary {
 systemName=NS_Cactus_Rocker_Female_Alt_Skin
 emitters [
  {emitterName=Trail_back, rendererClasses [&NiagaraMeshRendererProperties]}
  {emitterName=Trail_arms_and_legs, rendererClasses [&NiagaraMeshRendererProperties]}
  {emitterName=Particle_Decal, rendererClasses [&NiagaraSpriteRendererProperties]}
  {emitterName=Blocks_Right_Arm, rendererClasses [&NiagaraMeshRendererProperties]}
  {emitterName=Blocks_left_Ray, rendererClasses [&NiagaraMeshRendererProperties]}
  {emitterName=Blocks_Right_Ray, rendererClasses [&NiagaraMeshRendererProperties]}
  {emitterName=Blocks_left_Arm001, rendererClasses [&NiagaraMeshRendererProperties]}
 ]
 userVariables [
  User.bIsFrontEnd {type=NiagaraBool}
  User.bIsFrontEndPreview {type=NiagaraBool}
  User.bisHandHidden {type=NiagaraBool}
 ]
}
