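# Physics asset for the F_MED_Cactus_Rocker female cosmetic body, covering its dynamic shirt, strap, zipper and buckle bones with capsule bodies.
# Master bones for each accessory and the left thigh twist bone link to the pelvis with free angular constraints.
# Shirt, zipper and two-segment strap chains hang off their masters with limited constraints that mostly allow swing on one axis, while the buckle has only its master body.
# Larger capsules approximate the pelvis and left upper thigh.
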PhysicsAsset /BRCosmetics/Characters/Player/Female/Medium/Bodies/F_MED_Cactus_Rocker/Meshes/F_MED_Cactus_Rocker_Physics
bodies {
 Dyn_Master_Buckle [{shapeType=Capsule, length=1, radius=1, shapeName=Dyn_Master_Buckle_capsule}]
 Dyn_Master_Shirt [{shapeType=Capsule, length=0.5, radius=0.5, shapeName=Dyn_Master_Shirt_capsule}]
 Dyn_Master_Strap [{shapeType=Capsule, length=1, radius=1, shapeName=Dyn_Master_Strap_capsule}]
 Dyn_Master_Zipper [{
   shapeType=Capsule
   length=4
   radius=3
   shapeName=Dyn_Master_Zipper_capsule
   center=<0.6795 3.4949 0.7895>
   rotation=<rot 85 174.9997 179.9999>
  }]
 Dyn_Shirt [{shapeType=Capsule, length=10, radius=5, shapeName=Dyn_Shirt_capsule, center=<4.5 -4.2283 0.0092>, rotation=<rot 0.4304 -90 -1.7209>}]
 Dyn_Strap1 [{shapeType=Capsule, length=4, radius=0.5, shapeName=Dyn_Strap1_capsule, center=<2.5405 0 0>, rotation=<rot 90 0 0>}]
 Dyn_Strap2 [{
   shapeType=Capsule
   length=6.2063
   radius=1
   shapeName=Dyn_Strap2_capsule
   center=<2.6881 -0.1386 0.1638>
   rotation=<rot 1.9716 93.9949 -90.7661>
  }]
 Dyn_Zipper1 [{shapeType=Capsule, length=2, radius=0.505, shapeName=Dyn_Zipper1_capsule, center=<1.4529 0 0>, rotation=<rot 0.0933 84.3484 -88.9331>}]
 pelvis [{shapeType=Capsule, length=13, radius=8.1, shapeName=pelvis_capsule, center=<0.5311 1.2006 0>}]
 thigh_twist_01_l [
  {
   shapeType=Capsule
   length=10
   radius=8.1627
   shapeName=thigh_twist_01_l_capsule
   center=<16.8768 -0.1206 0.3811>
   rotation=<rot 85 174.3736 179.9999>
  }
  {shapeType=Capsule, length=10, radius=7, shapeName=thigh_twist_01_l_capsule, center=<3.7805 0.6969 0>, rotation=<rot 90 173.7581 170>}
 ]
}
constraints [
 {
  bone1Name=Dyn_Master_Shirt
  bone2Name=pelvis
  swing1LimitDegrees=45
  swing1Motion=Free
  swing2LimitDegrees=45
  swing2Motion=Free
  twistLimitDegrees=45
  twistMotion=Free
 }
 {
  bone1Name=thigh_twist_01_l
  bone2Name=pelvis
  swing1LimitDegrees=45
  swing1Motion=Free
  swing2LimitDegrees=45
  swing2Motion=Free
  twistLimitDegrees=45
  twistMotion=Free
 }
 {
  bone1Name=Dyn_Master_Strap
  bone2Name=pelvis
  swing1LimitDegrees=45
  swing1Motion=Free
  swing2LimitDegrees=45
  swing2Motion=Free
  twistLimitDegrees=45
  twistMotion=Free
 }
 {
  bone1Name=Dyn_Master_Zipper
  bone2Name=pelvis
  swing1LimitDegrees=45
  swing1Motion=Free
  swing2LimitDegrees=45
  swing2Motion=Free
  twistLimitDegrees=45
  twistMotion=Free
 }
 {
  bone1Name=Dyn_Master_Buckle
  bone2Name=pelvis
  swing1LimitDegrees=45
  swing1Motion=Free
  swing2LimitDegrees=45
  swing2Motion=Free
  twistLimitDegrees=45
  twistMotion=Free
 }
 {bone1Name=Dyn_Shirt, bone2Name=Dyn_Master_Shirt, swing1LimitDegrees=40, swing2LimitDegrees=1, twistLimitDegrees=1}
 {bone1Name=Dyn_Strap1, bone2Name=Dyn_Master_Strap, swing1LimitDegrees=40, swing2LimitDegrees=20, twistLimitDegrees=1}
 {bone1Name=Dyn_Strap2, bone2Name=Dyn_Strap1, swing1LimitDegrees=30, swing2LimitDegrees=10, twistLimitDegrees=1}
 {bone1Name=Dyn_Zipper1, bone2Name=Dyn_Master_Zipper, swing1LimitDegrees=30, swing2LimitDegrees=1, twistLimitDegrees=1}
]
