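# Animation Blueprint for the female Cactus Rocker body that adds deformation and secondary motion to the belt, buckle, strap, pouch and shirt.
# Standing and crouched branches copy the parent pose, run sculpter and twist Control Rigs, apply additive bone offsets, then blend on crouch.
# Left-thigh twist against the pelvis drives corrective alphas, with front-end and backbling checks adjusting belt and pouch.
# A pelvis-space RigidBody simulation finishes the pose.
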AnimBlueprint /BRCosmetics/Characters/Player/Female/Medium/Bodies/F_MED_Cactus_Rocker/Meshes/F_MED_Cactus_Rocker_AnimBP
target_skeleton=&/F_MED_Cactus_Rocker_Skeleton
graphs [
 AnimGraph {nodes [
   "Constraint - Dyn_Master_Buckle (R)"(Alpha=1, ConstraintWeights_0=1) {
    alphaBoolBlend {blendOption=Linear}
    alphaScaleBiasClamp {bMapRange=true}
    boneToModify {boneName=Dyn_Master_Buckle}
    constraintSetup [{transformType=Rotation, targetBone {boneName=Dyn_Master_Strap}}]
    constraintWeights [1]
   }
   "Constraint - Dyn_Master_Strap (R)"(Alpha=1, ConstraintWeights_0=1) {
    alphaBoolBlend {blendOption=Linear}
    boneToModify {boneName=Dyn_Master_Strap}
    constraintSetup [{offsetOption=None, transformType=Rotation, targetBone {boneName=Dyn_Master_Buckle}}]
    constraintWeights [1]
   }
   n1 = Control Rig(Source=Copy Pose From Mesh() {bUseAttachedParent=true}) {
    alphaBoolBlend {blendOption=Linear}
    controlRigAssetReference {blueprintRigClass=&/F_MED_Cactus_Rocker_Deform_CtrlRig_sculpters.F_MED_Cactus_Rocker_Deform_CtrlRig_sculpters_C}
   }
   n2 = Control Rig(Source=@n1) {
    alphaBoolBlend {blendOption=Linear}
    controlRigAssetReference {blueprintRigClass=&/F_MED_Cactus_Rocker_Deform_CtrlRig_twists.F_MED_Cactus_Rocker_Deform_CtrlRig_twists_C}
   }
   n3 = "Transform (Modify) Bone - Bone: Dyn_Master_Belt"(
    ComponentPose=Local To Component(LocalPose=@n2)
    Alpha=Get Left Leg Belt_Front_End() -> (Left Leg Belt_Front_End:double)
    Translation="1,0,0"
    Rotation="10,-5,0"
    Scale="1,1,1"
   ) {
    rotationMode=BMM_Additive
    rotationSpace=BCS_BoneSpace
    translationMode=BMM_Additive
    translationSpace=BCS_BoneSpace
    alphaBoolBlend {blendOption=Linear}
    alphaScaleBiasClamp {bInterpResult=true}
    boneToModify {boneName=Dyn_Master_Belt}
    rotation {pitch=10, yaw=-5}
    translation {x=1}
   }
   n4 = "Transform (Modify) Bone - Bone: Dyn_Master_Pouch"(ComponentPose=@n3, Alpha=Get Frontend() -> (Frontend:double), Translation="0,-4,-2", Rotation="0,0,-30", Scale="1,1,1") {
    rotationMode=BMM_Additive
    rotationSpace=BCS_BoneSpace
    translationMode=BMM_Additive
    translationSpace=BCS_BoneSpace
    alphaBoolBlend {blendOption=Linear}
    boneToModify {boneName=Dyn_Master_Pouch}
    rotation {roll=-30}
    translation {y=-4, z=-2}
   }
   n5 = "Transform (Modify) Bone - Bone: Dyn_Master_Belt"(ComponentPose=@n4, Alpha=Get Left Leg Belt() -> (Left Leg Belt:double), Translation="0,-0.5,0", Rotation="0,-15,0", Scale="1,1,1") {
    rotationMode=BMM_Additive
    rotationSpace=BCS_BoneSpace
    translationMode=BMM_Additive
    translationSpace=BCS_BoneSpace
    alphaBoolBlend {blendOption=Linear}
    alphaScaleBiasClamp {bInterpResult=true}
    boneToModify {boneName=Dyn_Master_Belt}
    rotation {yaw=-15}
    translation {y=-0.5}
   }
   n6 = "Transform (Modify) Bone - Bone: Dyn_Master_Shirt"(ComponentPose=@n5, Alpha=1, Translation="0,-1,0", Rotation="0,0,0", Scale="1,1,1") {
    rotationMode=BMM_Additive
    rotationSpace=BCS_BoneSpace
    translationMode=BMM_Additive
    translationSpace=BCS_BoneSpace
    alphaBoolBlend {blendOption=Linear}
    alphaScaleBiasClamp {bInterpResult=true}
    boneToModify {boneName=Dyn_Master_Shirt}
    translation {y=-1}
   }
   n7 = "Transform (Modify) Bone - Bone: Dyn_Master_Strap"(ComponentPose=@n6, Alpha=1, Translation="2.936,0.701,0.671", Rotation="0,0,0", Scale="1,1,1") {
    rotationMode=BMM_Additive
    rotationSpace=BCS_BoneSpace
    translationMode=BMM_Additive
    translationSpace=BCS_BoneSpace
    alphaBoolBlend {blendOption=Linear}
    alphaScaleBiasClamp {bInterpResult=true}
    boneToModify {boneName=Dyn_Master_Strap}
    translation=<2.936 0.701 0.671>
   }
   n8 = "Transform (Modify) Bone - Bone: Dyn_Master_Buckle"(ComponentPose=@n7, Alpha=1, Translation="0,-1,0", Rotation="0,-10,0", Scale="1,1,1") {
    rotationMode=BMM_Additive
    rotationSpace=BCS_BoneSpace
    translationMode=BMM_Additive
    translationSpace=BCS_BoneSpace
    alphaBoolBlend {blendOption=Linear}
    boneToModify {boneName=Dyn_Master_Buckle}
    rotation {yaw=-10}
    translation {y=-1}
   }
   n9 = "Transform (Modify) Bone - Bone: Dyn_Master_Buckle"(
    ComponentPose=@n8
    Alpha=Get Left Leg Strafe_Stand() -> (Left Leg Strafe_Stand:double)
    Translation="0,0,0"
    Rotation="-54.242,-15.455,-14.659"
    Scale="1,1,1"
   ) {
    rotationMode=BMM_Additive
    rotationSpace=BCS_BoneSpace
    translationMode=BMM_Additive
    translationSpace=BCS_BoneSpace
    alphaBoolBlend {blendOption=Linear}
    boneToModify {boneName=Dyn_Master_Buckle}
    rotation=<rot -54.242 -15.4552 -14.659>
   }
   n10 = Look At - Bone: Dyn_Master_Strap(ComponentPose=@n9, Alpha=1) {
    alphaBoolBlend {blendOption=Linear}
    alphaScaleBiasClamp {bInterpResult=true}
    boneToModify {boneName=Dyn_Master_Strap}
    lookAtLocation {y=100, z=100}
    lookAtTarget {boneReference {boneName=Dyn_Buckle1}}
   }
   n11 = Component To Local(ComponentPose=@n10)
   n12 = Control Rig(Source=Copy Pose From Mesh() {bUseAttachedParent=true}) {
    alphaBoolBlend {blendOption=Linear}
    controlRigAssetReference {blueprintRigClass=&F_MED_Cactus_Rocker_Deform_CtrlRig_sculpters_C}
   }
   n13 = Control Rig(Source=@n12) {
    alphaBoolBlend {blendOption=Linear}
    controlRigAssetReference {blueprintRigClass=&F_MED_Cactus_Rocker_Deform_CtrlRig_twists_C}
   }
   n14 = "Transform (Modify) Bone - Bone: Dyn_Master_Belt"(
    ComponentPose=Local To Component(LocalPose=@n13)
    Alpha=Get Left Leg Belt_Crouch() -> (Left Leg Belt_Crouch:double)
    Translation="0,0,0"
    Rotation="0.079,-13.788,-61.527"
    Scale="1,1,1"
   ) {
    rotationMode=BMM_Additive
    rotationSpace=BCS_BoneSpace
    translationMode=BMM_Additive
    translationSpace=BCS_BoneSpace
    alphaBoolBlend {blendOption=Linear}
    alphaScaleBiasClamp {bInterpResult=true}
    boneToModify {boneName=Dyn_Master_Belt}
    rotation=<rot 0.079 -13.7881 -61.527>
   }
   n15 = "Transform (Modify) Bone - Bone: Dyn_Master_Shirt"(ComponentPose=@n14, Alpha=1, Translation="0,-1,0", Rotation="0,0,0", Scale="1,1,1") {
    rotationMode=BMM_Additive
    rotationSpace=BCS_BoneSpace
    translationMode=BMM_Additive
    translationSpace=BCS_BoneSpace
    alphaBoolBlend {blendOption=Linear}
    alphaScaleBiasClamp {bInterpResult=true}
    boneToModify {boneName=Dyn_Master_Shirt}
    translation {y=-1}
   }
   n16 = "Transform (Modify) Bone - Bone: Dyn_Master_Buckle"(
    ComponentPose=@n15
    Alpha=Get Left Leg Strafe() -> (Left Leg Strafe:double)
    Translation="-4.630,2.333,-0.344"
    Rotation="-72.846,-73.973,-48.340"
    Scale="1,1,1"
   ) {
    rotationMode=BMM_Additive
    rotationSpace=BCS_BoneSpace
    translationMode=BMM_Additive
    translationSpace=BCS_BoneSpace
    alphaBoolBlend {blendOption=Linear}
    boneToModify {boneName=Dyn_Master_Buckle}
    rotation=<rot -72.846 -73.9731 -48.34>
    translation=<-4.63 2.333 -0.344>
   }
   n17 = "Transform (Modify) Bone - Bone: Dyn_Master_Shirt"(ComponentPose=@n16, Alpha=1, Translation="0,-2,0", Rotation="0,-10,10", Scale="1,1,1") {
    rotationMode=BMM_Additive
    rotationSpace=BCS_BoneSpace
    translationMode=BMM_Additive
    translationSpace=BCS_BoneSpace
    alphaBoolBlend {blendOption=Linear}
    boneToModify {boneName=Dyn_Master_Shirt}
    rotation {roll=10, yaw=-10}
    translation {y=-2}
   }
   n18 = "Transform (Modify) Bone - Bone: Dyn_Master_Strap"!(ComponentPose=@n17, bAlphaBoolEnabled:bool=Get bIsCrouching(), Translation="5.854,1.484,-1.673", Rotation="0,0,0", Scale="1,1,1") -> (Pose) {
    alphaInputType=Bool
    rotationMode=BMM_Additive
    rotationSpace=BCS_BoneSpace
    translationMode=BMM_Additive
    translationSpace=BCS_BoneSpace
    alphaBoolBlend {blendOption=Linear}
    boneToModify {boneName=Dyn_Master_Strap}
    translation=<5.854 1.484 -1.673>
   }
   n19 = Component To Local(ComponentPose=@n18)
   n20 = Get bIsCrouching()
   n21 = RigidBody(ComponentPose=Local To Component(LocalPose=Two Way Blend!(A=@n11, B=@n19, bAlphaBoolEnabled:bool=@n20) -> (Pose) {blendNode {alphaInputType=Bool, alphaBoolBlend {blendInTime=0.1, blendOption=Linear, blendOutTime=0.1}}})) {
    bClampLinearTranslationLimitToRefPose=true
    bOverrideWorldGravity=true
    overridePhysicsAsset=&/F_MED_Cactus_Rocker_Physics
    simulationSpace=BaseBoneSpace
    alphaBoolBlend {blendOption=Linear}
    baseBoneRef {boneName=pelvis}
    componentAppliedLinearAccClamp=<1000 1000 1000>
    componentLinearAccScale=<0.1 0.1 0.1>
    componentLinearVelScale=<0.1 0.1 0.1>
    overrideWorldGravity {z=-980}
   }
   Output Pose(Result=Component To Local(ComponentPose=@n21)) {layerGroup=None}
  ]}
 EventGraph {nodes [
   Event BlueprintUpdateAnimation()
   > n1 = Branch(Condition=Get bIsInFrontEnd())
   > n2 = Branch(Condition=CheckForBackbling())
   > Set Frontend(Frontend:double=0.4) -> (Output_Get:double)
   n3 = GetAlphaFromReferenceTransformDelta(SkeletalMeshComponent=Get Skel Mesh(), BoneName=pelvis, BaseName=thigh_l, InputRange=(X=10,Y=-20), TwistAxis="1,0,0")
   > Set Left Leg Belt_Front_End(Left Leg Belt_Front_End:double=@n3) -> (Output_Get:double)
   Set Left Leg Belt_Front_End(execute=@n2.else, Left Leg Belt_Front_End=@n3)
   > Set Frontend(Frontend=1)
   Set Frontend(execute=@n1.else, Frontend=0)
   > n4 = Branch(Condition=Get bIsCrouching())
   > Set Left Leg Strafe(Left Leg Strafe:double=GetAlphaFromReferenceTransformDelta(SkeletalMeshComponent=Get Skel Mesh(), BoneName=pelvis, BaseName=thigh_l, InputRange=(X=10,Y=80), TwistAxis="1,0,0")) -> (Output_Get:double)
   > Set Left Leg Strafe_Stand(Left Leg Strafe_Stand:double=0) -> (Output_Get:double)
   > Set Left Leg Belt(Left Leg Belt:double=0) -> (Output_Get:double)
   > Set Left Leg Belt_Crouch(Left Leg Belt_Crouch:double=GetAlphaFromReferenceTransformDelta(SkeletalMeshComponent=Get Skel Mesh(), BoneName=pelvis, BaseName=thigh_l, InputRange=(X=10,Y=80), TwistAxis="1,0,0")) -> (Output_Get:double)
   Set Left Leg Strafe(execute=@n4.else, Left Leg Strafe=0)
   > Set Left Leg Strafe_Stand(Left Leg Strafe_Stand=GetAlphaFromReferenceTransformDelta(SkeletalMeshComponent=Get Skel Mesh(), BoneName=pelvis, BaseName=thigh_l, InputRange=(X=13,Y=40), TwistAxis="1,0,0"))
   > Set Left Leg Belt(Left Leg Belt=GetAlphaFromReferenceTransformDelta(SkeletalMeshComponent=Get Skel Mesh(), BoneName=pelvis, BaseName=thigh_l, InputRange=(X=10,Y=27), TwistAxis="1,0,0"))
   > Set Left Leg Belt_Crouch(Left Leg Belt_Crouch=0)
   Event BlueprintInitializeAnimation()
   > Set Skel Mesh(Skel Mesh=GetOwnerSkeletalMeshComponent())
   Get Skel Mesh()
  ]}
]
variables [
 Alpha Output {type=number}
 Frontend {type=number}
 Left Leg Belt {type=number}
 Left Leg Belt_Crouch {type=number}
 Left Leg Belt_Front_End {type=number}
 Left Leg Strafe {type=number}
 Left Leg Strafe_Stand {type=number}
 Raw Output {type=number}
 Skel Mesh {type=SkeletalMeshComponent}
]
