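# Physics asset for the Cactus Rocker female character's FaceAcc mesh part, set up for dynamic hair.
# Capsules on spine, clavicles, neck and head form the body volume, with small master bodies joined to the head by free-angular constraints.
# Back, left and shoulder hair chains hang from those masters with swing limits and mostly tight twist.
# Cross-links tie neighbouring back strands together, some limited and some angularly free.
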PhysicsAsset /BRCosmetics/Characters/Player/Female/Medium/Bodies/F_MED_Cactus_Rocker/Meshes/Parts/F_MED_Cactus_Rocker_FaceAcc_Physics
bodies {
 Dyn_Master_Hair_bk [{shapeType=Capsule, length=1, radius=1, shapeName=Dyn_Master_Hair_bk_capsule}]
 clavicle_l [{
   shapeType=Capsule
   length=6.25
   radius=5.75
   shapeName=clavicle_l_capsule
   center=<5.2794 -1.1024 -1.5256>
   rotation=<rot -0.0009 -89.9999 -85>
  }]
 clavicle_r [{
   shapeType=Capsule
   length=14.5
   radius=5.75
   shapeName=clavicle_r_capsule
   center=<-7.8012 -0.2129 1.7379>
   rotation=<rot -0.0008 -89.9999 -85>
  }]
 dyn_hair_bk_01 [{shapeType=Capsule, length=5.3802, radius=2.0161, shapeName=dyn_hair_bk_01_capsule, center=<3.7237 0 0>, rotation=<rot 90 -162.646 -152.646>}]
 dyn_hair_bk_02 [{shapeType=Capsule, length=0.26, radius=2.5, shapeName=dyn_hair_bk_02_capsule, center=<2.8532 0 0>, rotation=<rot 90 0 0>}]
 dyn_hair_bk_03 [{shapeType=Capsule, length=1.3322, radius=4.3573, shapeName=dyn_hair_bk_03_capsule, center=<3.2673 0 0.4279>, rotation=<rot 90 0 0>}]
 dyn_hair_bk_04 [{shapeType=Capsule, length=2.0928, radius=3.3138, shapeName=dyn_hair_bk_04_capsule, center=<3.5966 0 2.4412>, rotation=<rot 90 0 0>}]
 dyn_hair_bk_05 [{shapeType=Capsule, length=3, radius=2.75, shapeName=dyn_hair_bk_05_capsule, center=<2.4738 0 1.2627>, rotation=<rot 90 0 0>}]
 dyn_hair_bk_r_00 [{shapeType=Capsule, length=4.3304, radius=2, shapeName=dyn_hair_bk_r_00_capsule, center=<4.124 -1.1196 -0.5246>, rotation=<rot 90 0 0>}]
 dyn_hair_bk_r_01 [{shapeType=Capsule, length=0.5, radius=3.5, shapeName=dyn_hair_bk_r_01_capsule, center=<3.4727 0.1379 0.2883>, rotation=<rot 90 0 0>}]
 dyn_hair_bk_r_02 [{shapeType=Capsule, length=1, radius=3.0325, shapeName=dyn_hair_bk_r_02_capsule, center=<1.8408 0 0>, rotation=<rot 90 0 0>}]
 dyn_hair_bk_r_03 [{shapeType=Capsule, length=2.4187, radius=1.8938, shapeName=dyn_hair_bk_r_03_capsule, center=<2.2213 0 0.7692>, rotation=<rot 90 0 0>}]
 dyn_hair_bk_sh_04 [{
   shapeType=Capsule
   length=3
   radius=2.25
   shapeName=dyn_hair_bk_sh_04_capsule
   center=<2.7285 0.3643 0.8652>
   rotation=<rot 79.9999 -0.0003 -0.0002>
  }]
 dyn_hair_bk_sh_l_00 [{shapeType=Capsule, length=6.8605, radius=2, shapeName=dyn_hair_bk_sh_l_00_capsule, center=<6.2923 -0.0189 0.2537>, rotation=<rot 90 0 0>}]
 dyn_hair_bk_sh_l_01 [{shapeType=Capsule, length=4.1565, radius=3.382, shapeName=dyn_hair_bk_sh_l_01_capsule, center=<1.6006 0 0>, rotation=<rot 90 0 0>}]
 dyn_hair_bk_sh_l_02 [{shapeType=Capsule, length=2.7758, radius=1.6914, shapeName=dyn_hair_bk_sh_l_02_capsule, center=<1.3577 0 0>, rotation=<rot 90 0 0>}]
 dyn_hair_bk_sh_l_03 [{shapeType=Capsule, length=4.25, radius=4.2, shapeName=dyn_hair_bk_sh_l_03_capsule, center=<4.4893 0 1.8307>, rotation=<rot 90 0 0>}]
 dyn_hair_bk_sh_l_04 [{shapeType=Capsule, length=5.01, radius=2.75, shapeName=dyn_hair_bk_sh_l_04_capsule, center=<2.7669 0 2.2295>, rotation=<rot 90 0 0>}]
 dyn_hair_bk_sh_r_00 [{shapeType=Capsule, length=3.5466, radius=2.5, shapeName=dyn_hair_bk_sh_r_00_capsule, center=<3.4087 0 0>, rotation=<rot 90 0 0>}]
 dyn_hair_bk_sh_r_01 [{shapeType=Capsule, length=3.5098, radius=3, shapeName=dyn_hair_bk_sh_r_01_capsule, center=<3.296 0 1.3861>, rotation=<rot 90 0 0>}]
 dyn_hair_bk_sh_r_02 [{shapeType=Capsule, length=3, radius=3, shapeName=dyn_hair_bk_sh_r_02_capsule, center=<3 0 0>, rotation=<rot 90 0 0>}]
 dyn_hair_bk_sh_r_03 [{shapeType=Capsule, length=3, radius=3, shapeName=dyn_hair_bk_sh_r_03_capsule, center=<3 0 0>, rotation=<rot 90 0 0>}]
 dyn_hair_bk_sh_r_04 [{shapeType=Capsule, length=3, radius=3, shapeName=dyn_hair_bk_sh_r_04_capsule, center=<3 0 0>, rotation=<rot 90 0 0>}]
 dyn_hair_fr_sh_l_01 [{
   shapeType=Capsule
   length=3.5456
   radius=1.5494
   shapeName=dyn_hair_fr_sh_l_01_capsule
   center=<2.9101 0 -0.3827>
   rotation=<rot 70.3123 -179.9999 -179.9999>
  }]
 dyn_hair_fr_sh_l_02 [{
   shapeType=Capsule
   length=5.755
   radius=1.505
   shapeName=dyn_hair_fr_sh_l_02_capsule
   center=<2.1901 0.6671 0.6401>
   rotation=<rot 81.5622 -160.0004 179.9997>
  }]
 dyn_hair_l_01 [{shapeType=Capsule, length=1.0566, radius=3.3408, shapeName=dyn_hair_l_01_capsule, center=<3.4052 0 -0.4835>, rotation=<rot 90 0 0>}]
 dyn_hair_l_02 [{shapeType=Capsule, length=4.4828, radius=1.9527, shapeName=dyn_hair_l_02_capsule, center=<2.506 0 0>, rotation=<rot 90 0 0>}]
 dyn_hair_l_MSTR [{shapeType=Capsule, length=1, radius=1, shapeName=dyn_hair_l_MSTR_capsule}]
 dyn_master_hair_bk_sh_l [{shapeType=Capsule, length=0.505, radius=0.505, shapeName=dyn_master_hair_bk_sh_l_capsule, rotation=<rot 0 -90 0>}]
 dyn_masterhair_bk_r [{shapeType=Capsule, length=0.505, radius=0.505, shapeName=dyn_masterhair_bk_r_capsule, rotation=<rot 0 -90 0>}]
 dyn_masterhair_bk_sh_r [{shapeType=Capsule, length=0.505, radius=0.505, shapeName=dyn_masterhair_bk_sh_r_capsule, rotation=<rot 0 -90 0>}]
 head [{shapeType=Capsule, length=1.75, radius=6.25, shapeName=head_capsule, center=<0 1.7687 0>}]
 neck_01 [{shapeType=Capsule, length=2.5, radius=4.5, shapeName=neck_01_capsule, rotation=<rot 0 -90 0>}]
 neck_02 [{shapeType=Capsule, length=1.505, radius=4.255, shapeName=neck_02_capsule, rotation=<rot 0 -90 0>}]
 spine_05 [{shapeType=Capsule, length=12.5, radius=6.5, shapeName=spine_05_capsule, center=<2.773 -1.1309 0>, rotation=<rot 0 -90 0>}]
}
constraints [
 {
  bone1Name=Dyn_Master_Hair_bk
  bone2Name=head
  swing1LimitDegrees=45
  swing1Motion=Free
  swing2LimitDegrees=45
  swing2Motion=Free
  twistLimitDegrees=45
  twistMotion=Free
 }
 {
  bone1Name=dyn_hair_l_MSTR
  bone2Name=head
  swing1LimitDegrees=45
  swing1Motion=Free
  swing2LimitDegrees=45
  swing2Motion=Free
  twistLimitDegrees=45
  twistMotion=Free
 }
 {
  bone1Name=head
  bone2Name=neck_02
  swing1LimitDegrees=45
  swing1Motion=Free
  swing2LimitDegrees=45
  swing2Motion=Free
  twistLimitDegrees=45
  twistMotion=Free
 }
 {
  bone1Name=neck_02
  bone2Name=neck_01
  swing1LimitDegrees=45
  swing1Motion=Free
  swing2LimitDegrees=45
  swing2Motion=Free
  twistLimitDegrees=45
  twistMotion=Free
 }
 {
  bone1Name=clavicle_l
  bone2Name=spine_05
  swing1LimitDegrees=45
  swing1Motion=Free
  swing2LimitDegrees=45
  swing2Motion=Free
  twistLimitDegrees=45
  twistMotion=Free
 }
 {
  bone1Name=clavicle_r
  bone2Name=spine_05
  swing1LimitDegrees=45
  swing1Motion=Free
  swing2LimitDegrees=45
  swing2Motion=Free
  twistLimitDegrees=45
  twistMotion=Free
 }
 {
  bone1Name=neck_01
  bone2Name=spine_05
  swing1LimitDegrees=45
  swing1Motion=Free
  swing2LimitDegrees=45
  swing2Motion=Free
  twistLimitDegrees=45
  twistMotion=Free
 }
 {bone1Name=dyn_hair_bk_02, bone2Name=dyn_hair_bk_01, swing1LimitDegrees=40, swing2LimitDegrees=40, twistLimitDegrees=1}
 {bone1Name=dyn_hair_bk_03, bone2Name=dyn_hair_bk_02, swing1LimitDegrees=40, swing2LimitDegrees=40, twistLimitDegrees=1}
 {bone1Name=dyn_hair_bk_04, bone2Name=dyn_hair_bk_03, swing1LimitDegrees=40, swing2LimitDegrees=40, twistLimitDegrees=1}
 {bone1Name=dyn_hair_bk_05, bone2Name=dyn_hair_bk_04, swing1LimitDegrees=40, swing2LimitDegrees=40, twistLimitDegrees=1}
 {bone1Name=dyn_hair_bk_sh_l_01, bone2Name=dyn_hair_bk_sh_l_00, swing1LimitDegrees=40, swing2LimitDegrees=40, twistLimitDegrees=1}
 {bone1Name=dyn_hair_bk_sh_l_02, bone2Name=dyn_hair_bk_sh_l_01, swing1LimitDegrees=20, swing2LimitDegrees=40, twistLimitDegrees=1}
 {bone1Name=dyn_hair_bk_sh_l_03, bone2Name=dyn_hair_bk_sh_l_02, swing1LimitDegrees=40, swing2LimitDegrees=40, twistLimitDegrees=1}
 {bone1Name=dyn_hair_bk_sh_l_04, bone2Name=dyn_hair_bk_sh_l_02, swing1LimitDegrees=40, swing2LimitDegrees=40, twistLimitDegrees=1}
 {bone1Name=dyn_hair_bk_sh_r_01, bone2Name=dyn_hair_bk_sh_r_00, swing1LimitDegrees=10, swing2LimitDegrees=60, twistLimitDegrees=1}
 {bone1Name=dyn_hair_l_01, bone2Name=dyn_hair_l_MSTR, swing1LimitDegrees=40, swing2LimitDegrees=40, twistLimitDegrees=1}
 {bone1Name=dyn_hair_l_02, bone2Name=dyn_hair_l_01, swing1LimitDegrees=40, swing2LimitDegrees=40, twistLimitDegrees=1}
 {bone1Name=dyn_hair_fr_sh_l_01, bone2Name=dyn_hair_l_02, swing1LimitDegrees=40, swing2LimitDegrees=40, twistLimitDegrees=1}
 {bone1Name=dyn_hair_fr_sh_l_02, bone2Name=dyn_hair_fr_sh_l_01, swing1LimitDegrees=40, swing2LimitDegrees=40, twistLimitDegrees=1}
 {bone1Name=dyn_hair_bk_r_01, bone2Name=dyn_hair_bk_r_00, swing1LimitDegrees=20, swing2LimitDegrees=40, twistLimitDegrees=1}
 {bone1Name=dyn_hair_bk_r_02, bone2Name=dyn_hair_bk_r_01, swing1LimitDegrees=30, swing2LimitDegrees=60, twistLimitDegrees=1}
 {bone1Name=dyn_hair_bk_r_03, bone2Name=dyn_hair_bk_r_02, swing1LimitDegrees=20, swing2LimitDegrees=60, twistLimitDegrees=5}
 {bone1Name=dyn_hair_bk_sh_04, bone2Name=dyn_hair_bk_r_03, swing1LimitDegrees=30, swing2LimitDegrees=60, twistLimitDegrees=1}
 {bone1Name=dyn_hair_bk_sh_l_00, bone2Name=dyn_master_hair_bk_sh_l, swing1LimitDegrees=40, swing2LimitDegrees=60, twistLimitDegrees=1}
 {bone1Name=dyn_hair_bk_sh_r_00, bone2Name=dyn_masterhair_bk_sh_r, swing1LimitDegrees=40, swing2LimitDegrees=40, twistLimitDegrees=20}
 {bone1Name=dyn_hair_bk_01, bone2Name=Dyn_Master_Hair_bk, swing1LimitDegrees=40, swing2LimitDegrees=40, twistLimitDegrees=1}
 {bone1Name=dyn_hair_bk_r_00, bone2Name=dyn_masterhair_bk_r, swing1LimitDegrees=20, swing2LimitDegrees=20, twistLimitDegrees=1}
 {
  bone1Name=dyn_masterhair_bk_r
  bone2Name=head
  swing1LimitDegrees=45
  swing1Motion=Free
  swing2LimitDegrees=45
  swing2Motion=Free
  twistLimitDegrees=45
  twistMotion=Free
 }
 {
  bone1Name=dyn_masterhair_bk_sh_r
  bone2Name=head
  swing1LimitDegrees=45
  swing1Motion=Free
  swing2LimitDegrees=45
  swing2Motion=Free
  twistLimitDegrees=45
  twistMotion=Free
 }
 {bone1Name=dyn_hair_bk_sh_r_02, bone2Name=dyn_hair_bk_sh_r_01, swing1LimitDegrees=10, swing2LimitDegrees=60, twistLimitDegrees=5}
 {bone1Name=dyn_hair_bk_sh_r_03, bone2Name=dyn_hair_bk_sh_r_02, swing1LimitDegrees=40, swing2LimitDegrees=60, twistLimitDegrees=5}
 {bone1Name=dyn_hair_bk_sh_r_04, bone2Name=dyn_hair_bk_sh_r_03, swing1LimitDegrees=40, swing2LimitDegrees=60, twistLimitDegrees=5}
 {bone1Name=dyn_hair_bk_r_00, bone2Name=dyn_hair_bk_sh_r_00, swing1LimitDegrees=30, swing2LimitDegrees=30, twistLimitDegrees=1}
 {bone1Name=dyn_hair_bk_r_03, bone2Name=dyn_hair_bk_sh_r_03, swing1LimitDegrees=20, swing2LimitDegrees=40, twistLimitDegrees=1}
 {
  bone1Name=dyn_master_hair_bk_sh_l
  bone2Name=head
  swing1LimitDegrees=45
  swing1Motion=Free
  swing2LimitDegrees=45
  swing2Motion=Free
  twistLimitDegrees=45
  twistMotion=Free
 }
 {
  bone1Name=dyn_hair_bk_sh_r_04
  bone2Name=dyn_hair_bk_05
  swing1LimitDegrees=45
  swing1Motion=Free
  swing2LimitDegrees=45
  swing2Motion=Free
  twistLimitDegrees=45
  twistMotion=Free
 }
 {
  bone1Name=dyn_hair_bk_04
  bone2Name=dyn_hair_bk_sh_l_03
  swing1LimitDegrees=45
  swing1Motion=Free
  swing2LimitDegrees=45
  swing2Motion=Free
  twistLimitDegrees=45
  twistMotion=Free
 }
]
