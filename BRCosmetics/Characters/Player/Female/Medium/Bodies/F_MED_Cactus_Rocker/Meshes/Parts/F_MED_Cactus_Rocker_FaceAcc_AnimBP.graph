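# Secondary-motion AnimBP for the Cactus Rocker face accessory (hat and hair), copying its pose from the parent mesh.
# Additive bone corrections feed neck-relative RigidBody simulations: a normal chain driven by head pitch, and a fixed-offset skydive chain with its own physics asset.
# A bool blend selects the skydive chain while skydiving without parachute.
# In the front end, the FrontEnd_Idle_Bypass curve layers DefaultSlot over the hat branch.
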AnimBlueprint /BRCosmetics/Characters/Player/Female/Medium/Bodies/F_MED_Cactus_Rocker/Meshes/Parts/F_MED_Cactus_Rocker_FaceAcc_AnimBP
target_skeleton=&/F_MED_Cactus_Rocker_FaceAcc_Skeleton
graphs [
 AnimGraph {nodes [
   Look At - Bone: dyn_hair_bk_r_02(Alpha=1) {
    alphaBoolBlend {blendOption=Linear}
    boneToModify {boneName=dyn_hair_bk_r_02}
    lookAtLocation {y=100, z=100}
    lookAtTarget {boneReference {boneName=Dyn_Right_Shoulder1}}
   }
   Output Pose(Result=Layered blend per bone(BasePose=Use cached pose 'PostDynamics'(), BlendPoses_0=Slot 'DefaultSlot'(), BlendWeights_0=Get Custom FE Idle() -> (Custom FE Idle:double)) {layerSetup [{branchFilters [{boneName=Dyn_Master_Hat}]}]}) {layerGroup=None}
   n1 = Get Head up() -> (Head up:double)
   n2 = "Transform (Modify) Bone - Bone: Dyn_Master_Hat"(
    ComponentPose=Local To Component(LocalPose=Copy Pose From Mesh() {bUseAttachedParent=true})
    Alpha=@n1
    Translation="0,0,0"
    Rotation="-0.000,-12,-0.000"
    Scale="1,1,1"
   ) {
    rotationMode=BMM_Additive
    rotationSpace=BCS_BoneSpace
    translationMode=BMM_Additive
    translationSpace=BCS_BoneSpace
    alphaBoolBlend {blendOption=Linear}
    boneToModify {boneName=Dyn_Master_Hat}
    rotation {yaw=-12}
   }
   n3 = "Transform (Modify) Bone - Bone: Dyn_Hat1"(ComponentPose=@n2, Alpha=@n1, Translation="0,0,0", Rotation="-0.000,-30,-0.000", Scale="1,1,1") {
    rotationMode=BMM_Additive
    rotationSpace=BCS_BoneSpace
    translationMode=BMM_Additive
    translationSpace=BCS_BoneSpace
    alphaBoolBlend {blendOption=Linear}
    boneToModify {boneName=Dyn_Hat1}
    rotation {yaw=-30}
   }
   n4 = "Transform (Modify) Bone - Bone: dyn_masterhair_bk_r"(ComponentPose=@n3, Alpha=@n1, Translation="0,0,0", Rotation="48.927,45.986,37.471", Scale="1,1,1") {
    rotationMode=BMM_Additive
    rotationSpace=BCS_BoneSpace
    translationMode=BMM_Additive
    translationSpace=BCS_BoneSpace
    alphaBoolBlend {blendOption=Linear}
    alphaScaleBiasClamp {interpSpeedIncreasing=20}
    boneToModify {boneName=dyn_masterhair_bk_r}
    rotation=<rot 48.9269 45.9859 37.4709>
   }
   n5 = "Transform (Modify) Bone - Bone: dyn_masterhair_bk_sh_r"(ComponentPose=@n4, Alpha=@n1, Translation="0,0,0", Rotation="53.802,69.269,72.959", Scale="1,1,1") {
    rotationMode=BMM_Additive
    rotationSpace=BCS_BoneSpace
    translationMode=BMM_Additive
    translationSpace=BCS_BoneSpace
    alphaBoolBlend {blendOption=Linear}
    boneToModify {boneName=dyn_masterhair_bk_sh_r}
    rotation=<rot 53.802 69.269 72.959>
   }
   n6 = "Transform (Modify) Bone - Bone: dyn_master_hair_bk_sh_l"(ComponentPose=@n5, Alpha=@n1, Translation="0,0,0", Rotation="40,-5,0", Scale="1,1,1") {
    rotationMode=BMM_Additive
    rotationSpace=BCS_BoneSpace
    translationMode=BMM_Additive
    translationSpace=BCS_BoneSpace
    alphaBoolBlend {blendOption=Linear}
    boneToModify {boneName=dyn_master_hair_bk_sh_l}
    rotation {pitch=40, yaw=-5}
   }
   n7 = "Transform (Modify) Bone - Bone: Dyn_Master_Hair_bk"(ComponentPose=@n6, Alpha=@n1, Translation="0,0,0", Rotation="0.000,-60,-0.000", Scale="1,1,1") {
    rotationMode=BMM_Additive
    rotationSpace=BCS_BoneSpace
    translationMode=BMM_Additive
    translationSpace=BCS_BoneSpace
    alphaBoolBlend {blendOption=Linear}
    boneToModify {boneName=Dyn_Master_Hair_bk}
    rotation {yaw=-60}
   }
   n8 = RigidBody(ComponentPose=@n7) {
    bClampLinearTranslationLimitToRefPose=true
    bOverrideWorldGravity=true
    overridePhysicsAsset=&/F_MED_Cactus_Rocker_FaceAcc_Physics
    simulationSpace=BaseBoneSpace
    alphaBoolBlend {blendOption=Linear}
    baseBoneRef {boneName=neck_02}
    componentAppliedLinearAccClamp=<750 750 750>
    componentLinearAccScale=<0.5 0.5 0.5>
    componentLinearVelScale=<0.5 0.5 0.5>
    overrideWorldGravity {z=-980}
   }
   n9 = "Transform (Modify) Bone - Bone: Dyn_Master_Hair_bk"(
    ComponentPose=Local To Component(LocalPose=Copy Pose From Mesh() {bUseAttachedParent=true})
    Alpha=1
    Translation="0,0,0"
    Rotation="-0.018,-60,0.009"
    Scale="1,1,1"
   ) {
    rotationMode=BMM_Additive
    rotationSpace=BCS_BoneSpace
    translationMode=BMM_Additive
    translationSpace=BCS_BoneSpace
    alphaBoolBlend {blendOption=Linear}
    boneToModify {boneName=Dyn_Master_Hair_bk}
    rotation=<rot -0.018 -60 0.009>
   }
   n10 = "Transform (Modify) Bone - Bone: dyn_hair_l_MSTR"(ComponentPose=@n9, Alpha=1, Translation="0,0,0", Rotation="-58.587,-30,-110", Scale="1,1,1") {
    rotationMode=BMM_Additive
    rotationSpace=BCS_BoneSpace
    translationMode=BMM_Additive
    translationSpace=BCS_BoneSpace
    alphaBoolBlend {blendOption=Linear}
    boneToModify {boneName=dyn_hair_l_MSTR}
    rotation=<rot -58.587 -30 -110>
   }
   n11 = "Transform (Modify) Bone - Bone: dyn_masterhair_bk_sh_r"(ComponentPose=@n10, Alpha=1, Translation="0,0,0", Rotation="52.288,47.062,-25.075", Scale="1,1,1") {
    rotationMode=BMM_Additive
    rotationSpace=BCS_BoneSpace
    translationMode=BMM_Additive
    translationSpace=BCS_BoneSpace
    alphaBoolBlend {blendOption=Linear}
    boneToModify {boneName=dyn_masterhair_bk_sh_r}
    rotation=<rot 52.288 47.062 -25.075>
   }
   n12 = "Transform (Modify) Bone - Bone: dyn_masterhair_bk_r"(ComponentPose=@n11, Alpha=1, Translation="0,0,0", Rotation="46.554,10.777,27.339", Scale="1,1,1") {
    rotationMode=BMM_Additive
    rotationSpace=BCS_BoneSpace
    translationMode=BMM_Additive
    translationSpace=BCS_BoneSpace
    alphaBoolBlend {blendOption=Linear}
    boneToModify {boneName=dyn_masterhair_bk_r}
    rotation=<rot 46.5539 10.7769 27.339>
   }
   n13 = "Transform (Modify) Bone - Bone: dyn_master_hair_bk_sh_l"(ComponentPose=@n12, Alpha=1, Translation="0,0,0", Rotation="49.420,-16.371,-17.435", Scale="1,1,1") {
    rotationMode=BMM_Additive
    rotationSpace=BCS_BoneSpace
    translationMode=BMM_Additive
    translationSpace=BCS_BoneSpace
    alphaBoolBlend {blendOption=Linear}
    boneToModify {boneName=dyn_master_hair_bk_sh_l}
    rotation=<rot 49.42 -16.371 -17.435>
   }
   n14 = "Transform (Modify) Bone - Bone: Dyn_Master_Hat"(ComponentPose=@n13, Alpha=1, Translation="0,0,0", Rotation="0,-12,0", Scale="1,1,1") {
    rotationMode=BMM_Additive
    rotationSpace=BCS_BoneSpace
    translationMode=BMM_Additive
    translationSpace=BCS_BoneSpace
    alphaBoolBlend {blendOption=Linear}
    boneToModify {boneName=Dyn_Master_Hat}
    rotation {yaw=-12}
   }
   n15 = RigidBody(ComponentPose=@n14) {
    bClampLinearTranslationLimitToRefPose=true
    bOverrideWorldGravity=true
    overridePhysicsAsset=&/F_MED_Cactus_Rocker_FaceAcc_Skydive_Physics
    simulationSpace=BaseBoneSpace
    alphaBoolBlend {blendOption=Linear}
    alphaScaleBiasClamp {bInterpResult=true}
    baseBoneRef {boneName=neck_02}
    componentAppliedLinearAccClamp=<750 750 750>
    componentLinearAccScale=<1 1 1>
    componentLinearVelScale=<1 1 1>
    overrideWorldGravity {z=-980}
   }
   Save cached pose 'PostDynamics'(Pose=Two Way Blend!(
     A=Component To Local(ComponentPose=@n8)
     B=Component To Local(ComponentPose=@n15)
     bAlphaBoolEnabled:bool=Get Skydive no Glide() -> (Skydive no Glide:bool)
    ) -> (Pose) {blendNode {alphaInputType=Bool, alphaBoolBlend {blendInTime=0.5, blendOption=Linear, blendOutTime=0.5}}}) {cacheName=PostDynamics}
  ]}
 EventGraph {nodes [
   Event BlueprintUpdateAnimation()
   > n1 = Sequence()
   > Set Head up(Head up:double=GetAlphaFromReferenceTransformDelta(SkeletalMeshComponent=Get Skel Mesh(), BoneName=head, BaseName=neck_02, InputRange=(X=-18,Y=-30), TwistAxis="1,0,0")) -> (Output_Get:double)
   n2 = Get bIsParachuteOpened()
   > Set Skydive no Glide(Skydive no Glide:bool=AND Boolean(A=NOT Boolean(A=@n2), B=Get bIsSkyDiving())) -> (Output_Get:bool)
   Branch(execute=@n1.then_1, Condition=Get bIsInFrontEnd())
   > Set Custom FE Idle(Custom FE Idle:double=GetCurveValue(self=GetOwnerFrontEndAnimInstance(), CurveName=FrontEnd_Idle_Bypass)) -> (Output_Get:double)
   Event BlueprintInitializeAnimation()
   > Set Skel Mesh(Skel Mesh=GetOwnerSkeletalMeshComponent())
  ]}
]
variables [
 Custom FE Idle {type=number}
 Head up {type=number}
 Skel Mesh {type=SkeletalMeshComponent}
 Skydive no Glide {type=boolean, default=true}
]
